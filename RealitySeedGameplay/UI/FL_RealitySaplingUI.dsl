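; Function library for Reality Sapling UI that converts fruit slots to byte values and looks up rarity display data.
; ConvertToFruitVariantByteArray rebuilds a byte array from the VariantOrRarity of each spawned slot's Tiered_RealityFruit_Athena fruit.
; GetRealitySaplingRarityData raises nurture level to a ScalableFloat minimum, casts it to EFortRarity and returns fields from that rarity's data, including Color2.
; Its FruitStates input is unused, and both branches return true.
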(fn ConvertToFruitVariantByteArray
  (params (FruitSlotArray :type Array%20of%20Server%20Spawned%20Fruit%20Slot%20Structures) (OutByteArray :type Array%20of%20Bytes))
  (Utilities|Array|Clear OutByteArray)
  (for _array_element FruitSlotArray
    (bind (_array_element_slot_4_56babd8945fe8b0c64a28cbc0719d0a6 _array_element_fruit_5_dcca3b6b43057342b19a728973abbf7e) (Utilities|Struct|BreakServerSpawnedFruitSlot _array_element))
    (Utilities|Array|Add (Variables|GetOutByteArray) (Variables|/RealitySeedGameplay/items/RealitySaplingFruits/Tiered_RealityFruit_Athena.Tiered_RealityFruit_Athena_C|GetVariantOrRarity _array_element_fruit_5_dcca3b6b43057342b19a728973abbf7e))))

(fn GetRealitySaplingRarityData
  (params (FruitStates :type Array%20of%20Bytes) (NurtureLevel :type Byte))
  (if (< NurtureLevel (ScalableFloat|GetValueasInteger (LocalVariables|GetMinNurturingLevel)))
    ;[1 1 1] Always clamp to the min nurturing level
    (bind _returnvalue (Fortnite|GetRarityData (Utilities|Enum|BytetoEnumEFortRarity (Math|Conversions|ToByte(Integer) (ScalableFloat|GetValueasInteger (LocalVariables|GetMinNurturingLevel))))))
    (return true (Utilities|Struct|BreakFortRarityItemData _returnvalue) (.Color2 (Utilities|Struct|BreakFortRarityItemData _returnvalue)))
    (else
      ;[1 1 1] If the nurture level is valid, just use it
      (bind _returnvalue_2 (Fortnite|GetRarityData (Utilities|Enum|BytetoEnumEFortRarity (Variables|GetNurtureLevel))))
      (return true (Utilities|Struct|BreakFortRarityItemData _returnvalue_2) (.Color2 (Utilities|Struct|BreakFortRarityItemData _returnvalue_2))))))
