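# Conversation marker owner widget showing a player's avatar, pin-colored diamond and name.
# InitPlayerState collapses it without a valid player state, otherwise async-loads the social icon into the avatar's dynamic material and shows SelfOwnedText for the local player, or a settings-filtered name for others.
# PreConstruct's enemy variant swaps the portrait material and hides the diamond.
# OverrideShowBackground drives SetBackgroundVisibility, which toggles the whole widget's visibility.
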WidgetBlueprint /FortniteConversationUI/Conversation/FortConversationMarker_OwnerInfo
graphs [
 EventGraph {nodes [
   n1 = InitPlayerState!() -> (OutputDelegate, then, PSA:Fort Player State Athena Object Reference, SocialIconTexture:Texture 2D Soft Object Reference) {is=title K2Node_CustomEvent}
   > SetVisibility(InVisibility=HitTestInvisible)
   > n2 = Is Valid(InputObject=@n1.PSA)
   > SetColorAndOpacity(
    self:Image Object Reference=Get Image_PingMarker() -> (Image_PingMarker:Image Object Reference)
    InColorAndOpacity:Linear Color Structure=GetPinColor!(self:Fort Player State Zone Object Reference=@n1.PSA) -> (ReturnValue:Linear Color Structure)
   )
   > n3 = Sequence()
   > n4 = Async Load Asset(Asset=@n1.SocialIconTexture)
   n5 = Cast To Texture2D(execute=[@n4.Completed, @n4.then], Object=@n4)
   > n6 = GetDynamicMaterial(self=Get Image_Avatar() -> (Image_Avatar:Fort Mobile Image Object Reference))
   > SetTextureParameterValue(self=@n6, ParameterName=SmallPreview, Value=@n5)
   n7 = Cast To FortPlayerControllerAthena(execute=@n3.then_1, Object=GetOwningPlayer())
   n8 = GetPlayerController!(self:Player State Object Reference=@n1.PSA) -> (ReturnValue:Player Controller Object Reference)
   n9 = GetOwningPlayer()
   > n10 = Branch(Condition=AND Boolean(A=IsValid(Object=@n8), B="Equal (Object)"(A=@n8, B=@n9)))
   n11 = Get Text_PlayerName() -> (Text_PlayerName:Common Text Object Reference)
   > "SetText (Text)"(self=@n11, InText=TextToUpper!(InText:"Text (by ref)"=Get SelfOwnedText() -> (SelfOwnedText:Text)) -> (ReturnValue:Text))
   "SetText (Text)"(
    execute=@n10.else
    self=@n11
    InText="To Text (String)"(InString=GetSettingsFilteredPlayerName!(self:Fort Player Controller Gameplay Object Reference=@n7, OtherPlayerState:Fort Player State Object Reference=@n1.PSA) -> (ReturnValue:String))
   )
   SetVisibility(execute=@n2.Is Not Valid, InVisibility=Collapsed)
   n12 = OverrideShowBackground!() -> (OutputDelegate, then, bShowBackground:bool=~) {is=title K2Node_CustomEvent}
   > SetBackgroundVisibility!(execute, self, bOverride:bool=true, bBackgroundVisible:bool=@n12.bShowBackground) -> (then)
   Event PreConstruct()
   > SetBrushFromMaterial(
    self=Get Image_Avatar()
    Material=Select(
     Option 0:Material Interface Object Reference=/Game/UI/SocialPanel/Art/Materials/UI_UserPortrait_Buffet_Small
     Option 1:Material Interface Object Reference=/Game/UI/SocialPanel/Art/Materials/UI_UserPortrait_Enemy
     Index=Get bEnemyVersion() -> (bEnemyVersion:bool)
    ) -> (ReturnValue:Material Interface Object Reference)
   )
   > SetVisibility(
    self=Get Image_PingMarker()
    InVisibility=Select(Option 0:ESlateVisibility Enum=HitTestInvisible, Option 1:ESlateVisibility Enum=Collapsed, Index=Get bEnemyVersion()) -> (ReturnValue:ESlateVisibility Enum)
   )
  ]}
 SetBackgroundVisibility {nodes [
   n1 = SetBackgroundVisibility!() -> (then, bOverride:bool=~, bBackgroundVisible:bool=~) {is=title K2Node_FunctionEntry}
   > n2 = Set bShowBackgroundOverridden(bShowBackgroundOverridden:bool=@n1) -> (Output_Get:bool)
   > n3 = Sequence()
   > n4 = Branch(Condition=@n2)
   > Set bLocal_BackgroundVisible(bLocal_BackgroundVisible:bool=Get bBackgroundVisible() -> (bBackgroundVisible:bool)) -> (Output_Get:bool)
   n5 = Get DBNOPulseTimer() -> (DBNOPulseTimer:Timer Handle Structure)
   Set bLocal_BackgroundVisible(
    execute=@n4.else
    bLocal_BackgroundVisible=OR Boolean!(
     A=Get bBackgroundVisible()
     B=Get bIsTalking() -> (bIsTalking:bool)
     C:bool=Is Valid Timer Handle!(Handle:Timer Handle Structure=@n5) -> (ReturnValue:bool)
    ) -> (ReturnValue)
   )
   SetVisibility(
    execute=@n3.then_1
    InVisibility=Select(Option 0=Collapsed, Option 1=SelfHitTestInvisible, Index=Get bLocal_BackgroundVisible() -> (bLocal_BackgroundVisible:bool))
   )
   Get WeakPSA() -> (WeakPSA:Fort Player State Athena Weak object)
  ]}
]
variables [
 bEnemyVersion {type=boolean}
 bIsTalking {type=boolean}
 bShowBackgroundOverridden {type=boolean}
 DBNOPulseTimer {type=TimerHandle}
 RelevantPlaylistTags {type=GameplayTagContainer, default {gameplayTags [{tagName=Athena.Playlist.Papaya}]}}
 SelfOwnedText {type=string, default=""}
]
widgets {
 info {inheritedWidgetCount=0, parentClass=&CommonUserWidget, rootWidgetClass=&Overlay, widgetCount=5}
 widgets [
  Overlay_0 {widgetClassPath=&Overlay, widget=Overlay_0}
  HorizontalBox_0 {widgetClassPath=&HorizontalBox, parent=Overlay_0, widget=HorizontalBox_0, slot {class=OverlaySlot}}
  Image_Avatar {
   widgetClassPath=/Script/FortniteUI.FortMobileImage
   bIsVariable=true
   parent=HorizontalBox_0
   widget=Image_Avatar
   properties {
    mobileSizeMultiplier=1.5
    visibility=HitTestInvisible
    brush {
     bIsDynamicallyLoaded=false
     drawAs=Image
     imageType=NoImage
     mirroring=NoMirror
     resourceName=None
     resourceObject=&UI_UserPortrait_Buffet_Small
     tiling=NoTile
     imageSize=<48 48>
     margin=<ltrb 0 0 0 0>
     outlineSettings {
      bUseBrushTransparency=false
      roundingType=HalfHeightRadius
      width=0
      color {colorUseRule=UseColor_Specified, specifiedColor=<rgba 0 0 0 0>}
      cornerRadii=<0 0 0 0>
     }
     tintColor {colorUseRule=UseColor_Specified, specifiedColor=<rgba 1 1 1 1>}
     uVRegion {bIsValid=false, max=<0 0>, min=<0 0>}
    }
   }
   slot {class=HorizontalBoxSlot, properties {horizontalAlignment=HAlign_Left, verticalAlignment=VAlign_Center}}
  }
  Image_PingMarker {
   widgetClassPath=&Image
   bIsVariable=true
   parent=HorizontalBox_0
   widget=Image_PingMarker
   properties {brush {
     bIsDynamicallyLoaded=false
     drawAs=Image
     imageType=NoImage
     mirroring=NoMirror
     resourceName=None
     resourceObject=/Game/Athena/HUD/Art/T_UI_Indicator_Diamond
     tiling=NoTile
     imageSize=<18 22>
     margin=<ltrb 0 0 0 0>
     outlineSettings {
      bUseBrushTransparency=false
      roundingType=HalfHeightRadius
      width=0
      color {colorUseRule=UseColor_Specified, specifiedColor=<rgba 0 0 0 0>}
      cornerRadii=<0 0 0 0>
     }
     tintColor {colorUseRule=UseColor_Specified, specifiedColor=<rgba 1 1 1 1>}
     uVRegion {bIsValid=false, max=<0 0>, min=<0 0>}
    }}
   slot {class=HorizontalBoxSlot, properties {horizontalAlignment=HAlign_Left, verticalAlignment=VAlign_Center, padding=<ltrb 0 0 2 0>}}
  }
  Text_PlayerName {
   widgetClassPath=&CommonTextBlock
   bIsVariable=true
   parent=HorizontalBox_0
   widget=Text_PlayerName
   properties {
    mobileFontSizeMultiplier=1.4
    text=Unknown
    font {
     bForceMonospaced=false
     bMaterialIsStencil=false
     fontMaterial=null
     fontObject=&Burbank
     letterSpacing=0
     monospacedWidth=1
     size=16
     skewAmount=0
     typefaceFontName=Burbank-PlayerPower
     outlineSettings {
      bApplyOutlineToDropShadows=false
      bMiteredCorners=false
      bSeparateFillAlpha=false
      outlineMaterial=&M-FontShadow
      outlineSize=3
      outlineColor=<rgba 0 0 0 0.8>
     }
    }
    shadowOffset=<0 0>
   }
   slot {class=HorizontalBoxSlot, properties {horizontalAlignment=HAlign_Left, verticalAlignment=VAlign_Center}}
  }
 ]
}
