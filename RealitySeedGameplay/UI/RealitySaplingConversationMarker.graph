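# Conversation marker widget for Reality Saplings that shows owner info in place of the participant name.
# When the indicated actor is set, it casts to the sapling and reads its owning player state and social avatar icon.
# A valid owner selects a teammate or enemy owner-info panel through a team check and switcher, otherwise the name text shows.
# It binds to the sapling's destroyed-for-transplant event to collapse itself.
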WidgetBlueprint /RealitySeedGameplay/UI/RealitySaplingConversationMarker
graphs [EventGraph {nodes [
   n1 = CustomEvent() {is=title K2Node_CustomEvent}
   > SetVisibility(InVisibility=Collapsed)
   n2 = Event OnSetIndicatedActor!() -> (OutputDelegate, then, NewIndicatorActor:Actor Object Reference) {is=title K2Node_Event}
   > n3 = Cast To B_SMA_Athena_RealitySapling_Parent!(execute, Object=@n2.NewIndicatorActor) -> (then, CastFailed, AsB SMA Athena Reality Sapling Parent:B SMA Athena Reality Sapling Parent Object Reference) {is=title K2Node_DynamicCast}
   > Bind Event to Sapling Destroyed for Transplant Event(self:B SMA Athena Reality Sapling Parent Object Reference=@n3, Delegate=@n1) {is=title K2Node_AddDelegate}
   n4 = Get Owning Player State Repped!(self:B SMA Athena Reality Sapling Parent Object Reference=@n3) -> (Owning Player State Repped:Fort Player State Athena Object Reference)
   > n5 = Is Valid(InputObject=@n4)
   SetVisibility(
    execute=[@n5.Is Not Valid, @n3.CastFailed]
    self=Get Switcher_MarkerStyle() -> (Switcher_MarkerStyle:Widget Switcher Object Reference)
    InVisibility=Hidden
   )
   > SetVisibility(self=Get Text_ParticipantName() -> (Text_ParticipantName:Common Text Object Reference), InVisibility=HitTestInvisible)
   SetVisibility(execute=@n5.Is Valid, self=Get Switcher_MarkerStyle(), InVisibility=SelfHitTestInvisible)
   > SetVisibility(self=Get Text_ParticipantName(), InVisibility=Collapsed)
   > n6 = Branch(Condition=OnSameTeam!(ActorA:Actor Object Reference=GetOwningPlayer(), ActorB:Actor Object Reference=@n4) -> (ReturnValue:bool))
   n7 = Get OwningPlayerSocialAvatarIcon!(self:B SMA Athena Reality Sapling Parent Object Reference=@n3) -> (OwningPlayerSocialAvatarIcon:Texture 2D Soft Object Reference)
   > InitPlayerState!(
    execute
    self:Fort Conversation Marker Owner Info Object Reference=Get RealitySaplingMarkerInfo() -> (RealitySaplingMarkerInfo:Fort Conversation Marker Owner Info Object Reference)
    PSA:Fort Player State Athena Object Reference=@n4
    SocialIconTexture:Texture 2D Soft Object Reference=@n7
   ) -> (then)
   n8 = Get Switcher_MarkerStyle()
   > SetActiveWidgetIndex(self=@n8, Index=0)
   InitPlayerState(
    execute=@n6.else
    self=Get RealitySaplingMarkerInfo_Enemy() -> (RealitySaplingMarkerInfo_Enemy:Fort Conversation Marker Owner Info Object Reference)
    PSA=@n4
    SocialIconTexture=@n7
   )
   > SetActiveWidgetIndex(self=@n8, Index=1)
  ]}]
variables []
widgets {
 info {
  inheritedWidgetCount=7
  parentClass=/Script/FortniteConversationUI.FortConversationMarker
  rootWidgetClass=&CommonVisibilitySwitcher
  widgetCount=24
 }
 widgets [
  Switcher_States {widgetClassPath=&CommonVisibilitySwitcher, bInherited=true, widget=Switcher_States}
  Overlay_PreviewState {
   widgetClassPath=&Overlay
   bInherited=true
   parent=Switcher_States
   widget=Overlay_PreviewState
   slot {class=CommonVisibilitySwitcherSlot}
  }
  VerticalBox_2 {
   widgetClassPath=&VerticalBox
   parent=Overlay_PreviewState
   widget=VerticalBox_2
   slot {class=OverlaySlot, properties {horizontalAlignment=HAlign_Center}}
  }
  SizeBox_0 {
   widgetClassPath=&SizeBox
   parent=VerticalBox_2
   widget=SizeBox_0
   properties {bOverride_HeightOverride=true, bOverride_WidthOverride=true, heightOverride=40, widthOverride=40}
   slot {class=VerticalBoxSlot, properties {horizontalAlignment=HAlign_Center, verticalAlignment=VAlign_Center}}
  }
  NPCIcon {
   widgetClassPath=&Image
   bInherited=true
   bIsVariable=true
   parent=SizeBox_0
   widget=NPCIcon
   properties {brush {
     bIsDynamicallyLoaded=false
     drawAs=Image
     imageType=NoImage
     mirroring=NoMirror
     resourceName=None
     resourceObject=/FortniteConversationUI/Conversation/Materials/T_UI_NPC_SpeechIcon
     tiling=NoTile
     imageSize=<32 32>
     margin=<ltrb 0 0 0 0>
     outlineSettings {
      bUseBrushTransparency=false
      roundingType=HalfHeightRadius
      width=0
      color {colorUseRule=UseColor_Specified, specifiedColor=<rgba 0 0 0 0>}
      cornerRadii=<0 0 0 0>
     }
     tintColor {colorUseRule=UseColor_Specified, specifiedColor=<rgba 1 1 1 1>}
     uVRegion {bIsValid=false, max=<0 0>, min=<0 0>}
    }}
   slot {class=SizeBoxSlot}
  }
  VerticalBox_0 {
   widgetClassPath=&VerticalBox
   parent=Switcher_States
   widget=VerticalBox_0
   slot {class=CommonVisibilitySwitcherSlot, properties {horizontalAlignment=HAlign_Center, verticalAlignment=VAlign_Center}}
  }
  Callout {
   widgetClassPath=&Overlay
   bIsVariable=true
   parent=VerticalBox_0
   widget=Callout
   properties {visibility=Collapsed}
   slot {class=VerticalBoxSlot, properties {horizontalAlignment=HAlign_Center, padding=<ltrb 0 0 0 10>}}
  }
  HorizontalBox_38 {widgetClassPath=&HorizontalBox, parent=Callout, widget=HorizontalBox_38, slot {class=OverlaySlot}}
  CommonTextBlock_74 {
   widgetClassPath=&CommonTextBlock
   parent=HorizontalBox_38
   widget=CommonTextBlock_74
   properties {
    text=Hold
    font {
     bForceMonospaced=false
     bMaterialIsStencil=false
     fontMaterial=null
     fontObject=&BurbankSmall
     letterSpacing=0
     monospacedWidth=1
     size=16
     skewAmount=0
     typefaceFontName=Bold
     outlineSettings {
      bApplyOutlineToDropShadows=false
      bMiteredCorners=false
      bSeparateFillAlpha=false
      outlineMaterial=null
      outlineSize=0
      outlineColor=<rgba 0 0 0 1>
     }
    }
    shadowOffset=<0 0>
   }
   slot {class=HorizontalBoxSlot, properties {verticalAlignment=VAlign_Center, padding=<ltrb 0 0 10 0>}}
  }
  KeybindWidget {
   widgetClassPath=/FortUILibrary/Input/KeybindWidget
   bIsVariable=true
   parent=HorizontalBox_38
   widget=KeybindWidget
   properties {boundAction=Target}
   slot {class=HorizontalBoxSlot}
  }
  CommonTextBlock {
   widgetClassPath=&CommonTextBlock
   parent=HorizontalBox_38
   widget=CommonTextBlock
   properties {
    text=To Interact
    font {
     bForceMonospaced=false
     bMaterialIsStencil=false
     fontMaterial=null
     fontObject=&BurbankSmall
     letterSpacing=0
     monospacedWidth=1
     size=16
     skewAmount=0
     typefaceFontName=Bold
     outlineSettings {
      bApplyOutlineToDropShadows=false
      bMiteredCorners=false
      bSeparateFillAlpha=false
      outlineMaterial=null
      outlineSize=0
      outlineColor=<rgba 0 0 0 1>
     }
    }
    shadowOffset=<0 0>
   }
   slot {class=HorizontalBoxSlot, properties {verticalAlignment=VAlign_Center, padding=<ltrb 10 0 0 0>}}
  }
  ConversationMessage_Main {
   widgetClassPath=/FortniteConversationUI/Conversation/FortMessageBubbleNew
   bInherited=true
   bIsVariable=true
   parent=VerticalBox_0
   widget=ConversationMessage_Main
   slot {class=VerticalBoxSlot, properties {horizontalAlignment=HAlign_Center, padding=<ltrb 0 0 0 30>}}
  }
  VerticalBox_1 {widgetClassPath=&VerticalBox, parent=Switcher_States, widget=VerticalBox_1, slot {class=CommonVisibilitySwitcherSlot}}
  HorizontalBox {widgetClassPath=&HorizontalBox, parent=VerticalBox_1, widget=HorizontalBox, slot {class=VerticalBoxSlot}}
  Overlay_2 {
   widgetClassPath=&Overlay
   parent=HorizontalBox
   widget=Overlay_2
   properties {visibility=Collapsed}
   slot {class=HorizontalBoxSlot}
  }
  Keybind_Nameplate {
   widgetClassPath=&KeybindWidget
   bInherited=true
   bIsVariable=true
   parent=Overlay_2
   widget=Keybind_Nameplate
   properties {boundAction=Use, visibility=HitTestInvisible}
   slot {class=OverlaySlot}
  }
  Overlay_0 {widgetClassPath=&Overlay, parent=HorizontalBox, widget=Overlay_0, slot {class=HorizontalBoxSlot}}
  Text_ParticipantName {
   widgetClassPath=&CommonTextBlock
   bInherited=true
   bIsVariable=true
   parent=Overlay_0
   widget=Text_ParticipantName
   properties {
    visibility=HitTestInvisible
    font {
     bForceMonospaced=false
     bMaterialIsStencil=false
     fontMaterial=null
     fontObject=&Burbank
     letterSpacing=0
     monospacedWidth=1
     size=14
     skewAmount=0
     typefaceFontName=Burbank-PlayerPower
     outlineSettings {
      bApplyOutlineToDropShadows=false
      bMiteredCorners=false
      bSeparateFillAlpha=false
      outlineMaterial=&M-FontShadow
      outlineSize=3
      outlineColor=<rgba 0 0 0 0.8>
     }
    }
    renderTransform {angle=0, scale=<1 1>, shear=<0 0>, translation=<0 -15>}
    shadowOffset=<0 0>
   }
   slot {class=OverlaySlot, properties {padding=<ltrb 75 0 0 0>}}
  }
  Switcher_MarkerStyle {widgetClassPath=&WidgetSwitcher, bIsVariable=true, parent=Overlay_0, widget=Switcher_MarkerStyle, slot {class=OverlaySlot}}
  RealitySaplingMarkerInfo {
   widgetClassPath=/FortniteConversationUI/Conversation/FortConversationMarker_OwnerInfo
   bIsVariable=true
   parent=Switcher_MarkerStyle
   widget=RealitySaplingMarkerInfo
   properties {selfOwnedText=Your sapling}
   slot {class=WidgetSwitcherSlot}
  }
  RealitySaplingMarkerInfo_Enemy {
   widgetClassPath=&FortConversationMarker_OwnerInfo
   bIsVariable=true
   parent=Switcher_MarkerStyle
   widget=RealitySaplingMarkerInfo_Enemy
   properties {bEnemyVersion=true}
   slot {class=WidgetSwitcherSlot}
  }
  CommonVisualAttachment_39 {
   widgetClassPath=&CommonVisualAttachment
   parent=VerticalBox_1
   widget=CommonVisualAttachment_39
   properties {contentAnchor=<0.5 0.7463>}
   slot {class=VerticalBoxSlot, properties {horizontalAlignment=HAlign_Center, verticalAlignment=VAlign_Center}}
  }
  Image_4 {
   widgetClassPath=&Image
   bIsVariable=true
   parent=CommonVisualAttachment_39
   widget=Image_4
   properties {brush {
     bIsDynamicallyLoaded=false
     drawAs=Image
     imageType=NoImage
     mirroring=NoMirror
     resourceName=None
     resourceObject=/Game/Athena/HUD/LTM/Mash/Materials/M_UI_InteractionArrow
     tiling=NoTile
     imageSize=<128 128>
     margin=<ltrb 0 0 0 0>
     outlineSettings {
      bUseBrushTransparency=false
      roundingType=HalfHeightRadius
      width=0
      color {colorUseRule=UseColor_Specified, specifiedColor=<rgba 0 0 0 0>}
      cornerRadii=<0 0 0 0>
     }
     tintColor {colorUseRule=UseColor_Specified, specifiedColor=<rgba 1 1 1 1>}
     uVRegion {bIsValid=false, max=<0 0>, min=<0 0>}
    }}
   slot {class=SizeBoxSlot}
  }
  TutorialHighlight {widgetClassPath=&TutorialHighlight}
 ]
}
