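# Translucent UI material for the Reality Fruit map icon, showing ready, picked or not-ripe status from a Fruit_State scalar.
# One icon texture packs stem mask, fruit mask and fill gradient into RGB channels, used to layer falloff, rarity and stem colors.
# An unripe color derives from the rarity color by lowering saturation and value in HSV, then blending back.
# An If on the state dims opacity when picked.
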Material /RealitySeedGameplay/UI/Icons/Map/M_UI_RealityFruitStatus
graph {
 properties {blendMode=BLEND_Translucent, materialDomain=MD_UI, pixelDepthOffsetMode=PDOM_Legacy, usedShadingModels=Default Lit}
 nodes [
  n1 = VectorParameter() {parameterName=RGB_FruitFalloff, defaultValue=<rgba 0.0312 0.0312 0.0312 1>}
  n2 = VectorParameter() {parameterName=RGB_Rarity, defaultValue=<rgba 0.2708 0.2708 0.2708 1>}
  n3 = NamedRerouteUsage()
  n4 = NamedRerouteUsage()
  LinearInterpolate(
   A=LinearInterpolate(A=LinearInterpolate(A=@n1.RGB, B=@n2.RGB, Alpha=@n3), B=LinearInterpolate(A=@n1.RGB, B=@n4, Alpha=@n3), Alpha=NamedRerouteUsage()) {constA=1}
   B=VectorParameter() {parameterName=RGB_Stems, defaultValue=<rgba 1 1 1 1>}.RGB
   Alpha=NamedRerouteUsage()
  )
  n5 = TextureSample() {samplerType=SAMPLERTYPE_LinearColor, texture=&/T_UI_RealityFruit_Icon_64x}
  n6 = NamedRerouteDeclaration(None=@n5.R) {name=StemsMask, nodeColor=<rgba 0 0.7188 1 1>}
  n7 = NamedRerouteDeclaration(None=@n5.G) {name=FruitMask, nodeColor=<rgba 1 0.2578 0 1>}
  NamedRerouteDeclaration(None=Saturate(None=Add(A=@n6, B=@n7))) {name=FullAlpha, nodeColor=<rgba 0 0.6484 1 1>}
  n8 = DotProduct(A=@n7, B=LinearInterpolate(A=ScalarParameter() {defaultValue=1, parameterName=FruitFillLift}, Alpha=@n5.B))
  NamedRerouteDeclaration(None=DotProduct(A=OneMinus(None=@n6), B=Saturate(None=@n8))) {name=MaskedFruitMaskGradient, nodeColor=<rgba 1 0.6563 0 1>}
  n9 = ScalarParameter() {desc="0 Ready, 1 Picked, 2 Not Ripe", parameterName=Fruit_State}
  NamedRerouteDeclaration(None=Saturate(None=Subtract(A=@n9))) {name=FruitStateSwitch, nodeColor=<rgba 0 0.7891 1 1>}
  n10 = VectorParameter() {parameterName=RGB_Rarity, defaultValue=<rgba 0.2708 0.2708 0.2708 1>}
  n11 = MaterialFunctionCall(Float3=MaterialFunctionCall!(In=@n10.RGB) {materialFunction=/Engine/Functions/Engine_MaterialFunctions01/Chromakeying/RGBtoHSV}.Result) {materialFunction=&BreakOutFloat3Components}
  n12 = MaterialFunctionCall(R=@n11.R, G=Multiply(A=@n11.G) {constB=0.75}, B=Multiply(A=@n11.B) {constB=0.025}) {materialFunction=&AppendMany}
  NamedRerouteDeclaration(None=LinearInterpolate(
    A=@n10.RGB
    B=MaterialFunctionCall!(HSV_In=@n12.RGB) {materialFunction=/Game/Athena/UI/Materials/MF_HSVtoRGB}.RGB_Out
    Alpha=ScalarParameter() {defaultValue=0.5, parameterName=UnripeMixAmount}
   )) {name=UnripeColor, nodeColor=<rgba 0 0.5078 1 1>}
  NamedRerouteDeclaration(None=DotProduct(A=OneMinus(None=@n6), B=NamedRerouteUsage())) {name=MaskedFruitMaskSolid, nodeColor=<rgba 0 0.0859 1 1>}
  n13 = ScalarParameter() {desc="0 Ready, 1 Picked, 2 Not Ripe", parameterName=Fruit_State}
  n14 = Constant() {r=1}
  n15 = ScalarParameter() {defaultValue=1, parameterName=ReadyAlpha}
  n16 = ScalarParameter() {defaultValue=0.333, parameterName=PickedAlpha}
  n17 = Multiply(A=@n16, B=LinearInterpolate(B=@n16, Alpha=NamedRerouteUsage()) {constA=1})
  Saturate(None=DotProduct(A=NamedRerouteUsage(), B=If(A=@n13, B=@n14, "A > B"=@n15, "A == B"=@n17, "A < B"=@n15)))
 ]
}
