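# UMG icon widget that shows a Reality Fruit's status, with an icon named for map use.
# An overlay stacks a sizing spacer, an optional offset dark shadow image and the status image.
# Construct and design-time PreConstruct toggle the shadow from UseShadow, with IconSize intended to size the spacer.
# SetPicked, SetReady and SetRarityColor drive the status image's dynamic material through its Fruit_State and RGB_Rarity parameters.
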WidgetBlueprint /RealitySeedGameplay/UI/RealityFruit
graphs [
 EventGraph {nodes [
   n1 = Event PreConstruct()
   > n2 = Branch(Condition=@n1.IsDesignTime)
   SetVisibility(
    execute=[@n4.then, @n2.then]
    self=Get Image_FruitStatus_OptionalShadow() -> (Image_FruitStatus_OptionalShadow:Image Object Reference)
    InVisibility=Select(Option 0=Collapsed, Option 1=SelfHitTestInvisible, Index=Get UseShadow() -> (UseShadow:bool))
   )
   n3 = Get IconSize() -> (IconSize:double)
   > SetSize!(execute, self:Spacer Object Reference=Get Size() -> (Size:Spacer Object Reference), InSize:Vector 2D Structure=MakeVector2D(X=@n3, Y=@n3)) -> (then)
   n4 = Event Construct()
   Event Tick()
  ]}
 SetPicked {nodes [
   n1 = SetPicked!() -> (then, Picked:bool=~) {is=title K2Node_FunctionEntry}
   > Branch(Condition=@n1)
   > n2 = GetDynamicMaterial(self=Get Image_FruitStatus() -> (Image_FruitStatus:Image Object Reference))
   > SetScalarParameterValue(self=@n2, ParameterName=Fruit_State, Value=1)
  ]}
 SetRarityColor {nodes [
   n1 = SetRarityColor!() -> (then, Color:Linear Color Structure) {is=title K2Node_FunctionEntry}
   > n2 = GetDynamicMaterial(self=Get Image_FruitStatus())
   > SetVectorParameterValue(self=@n2, ParameterName=RGB_Rarity, Value=@n1)
  ]}
 SetReady {nodes [
   n1 = SetReady!() -> (then, Ready:bool=~) {is=title K2Node_FunctionEntry}
   > n2 = GetDynamicMaterial(self=Get Image_FruitStatus())
   > SetScalarParameterValue(self=@n2, ParameterName=Fruit_State, Value=SelectFloat(A=0, B=2, bPickA=@n1))
  ]}
]
variables [
 IconSize {type=number, default=32}
 UseShadow {type=boolean}
]
widgets {
 info {inheritedWidgetCount=0, parentClass=&UserWidget, rootWidgetClass=&Overlay, widgetCount=4}
 widgets [
  Overlay_0 {widgetClassPath=&Overlay, widget=Overlay_0}
  Size {
   widgetClassPath=&Spacer
   bIsVariable=true
   parent=Overlay_0
   widget=Size
   properties {size=<32 32>}
   slot {class=OverlaySlot, properties {horizontalAlignment=HAlign_Fill, verticalAlignment=VAlign_Fill}}
  }
  Image_FruitStatus_OptionalShadow {
   widgetClassPath=&Image
   bIsVariable=true
   parent=Overlay_0
   widget=Image_FruitStatus_OptionalShadow
   properties {
    renderOpacity=0.125
    visibility=Collapsed
    brush {
     bIsDynamicallyLoaded=false
     drawAs=Image
     imageType=NoImage
     mirroring=NoMirror
     resourceName=None
     resourceObject=/RealitySeedGameplay/UI/Icons/Map/M_UI_RealityFruitStatus_Shadow
     tiling=NoTile
     imageSize=<32 32>
     margin=<ltrb 0 0 0 0>
     outlineSettings {
      bUseBrushTransparency=false
      roundingType=HalfHeightRadius
      width=0
      color {colorUseRule=UseColor_Specified, specifiedColor=<rgba 0 0 0 0>}
      cornerRadii=<0 0 0 0>
     }
     tintColor {colorUseRule=UseColor_Specified, specifiedColor=<rgba 1 1 1 1>}
     uVRegion {bIsValid=false, max=<0 0>, min=<0 0>}
    }
    colorAndOpacity=<rgba 0 0 0 1>
    renderTransform {angle=0, scale=<1 1>, shear=<0 0>, translation=<1.5 1.5>}
   }
   slot {class=OverlaySlot, properties {horizontalAlignment=HAlign_Fill, verticalAlignment=VAlign_Fill}}
  }
  Image_FruitStatus {
   widgetClassPath=&Image
   bIsVariable=true
   parent=Overlay_0
   widget=Image_FruitStatus
   properties {brush {
     bIsDynamicallyLoaded=false
     drawAs=Image
     imageType=NoImage
     mirroring=NoMirror
     resourceName=None
     resourceObject=/RealitySeedGameplay/UI/Icons/Map/M_UI_RealityFruitStatus
     tiling=NoTile
     imageSize=<32 32>
     margin=<ltrb 0 0 0 0>
     outlineSettings {
      bUseBrushTransparency=false
      roundingType=HalfHeightRadius
      width=0
      color {colorUseRule=UseColor_Specified, specifiedColor=<rgba 0 0 0 0>}
      cornerRadii=<0 0 0 0>
     }
     tintColor {colorUseRule=UseColor_Specified, specifiedColor=<rgba 1 1 1 1>}
     uVRegion {bIsValid=false, max=<0 0>, min=<0 0>}
    }}
   slot {class=OverlaySlot, properties {horizontalAlignment=HAlign_Fill, verticalAlignment=VAlign_Fill}}
  }
 ]
}
