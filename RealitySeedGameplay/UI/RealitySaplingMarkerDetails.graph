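# Marker details widget for a pinged Reality Sapling, showing owner name, rarity color, ready fruit count and weed or destroyed status.
# SetMarkerData caches the sapling, fills these fields from it, and tints the separator, dynamic icon material and parent marker line.
# Enemy harvest or damage delegates switch it to a red alert that hides fruit.
# It also restores that alert from the ping marker's stored enemy event state.
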WidgetBlueprint /RealitySeedGameplay/UI/RealitySaplingMarkerDetails
graphs [
 EnableEnemyEventState {nodes [
   n1 = EnableEnemyEventState!() -> (then, IsEnemyEventActive:bool=~, StatusIndex:int) {is=title K2Node_FunctionEntry}
   > n2 = Set Is Enemy Event Active(Is Enemy Event Active:bool=@n1) -> (Output_Get:bool)
   > Branch(Condition=@n2)
   > SetActiveWidgetIndex(self=Get WidgetSwitcher_Status() -> (WidgetSwitcher_Status:Widget Switcher Object Reference), Index=@n1.StatusIndex)
   > SetActiveWidgetIndex(self=Get Switcher_IndicatorIcon() -> (Switcher_IndicatorIcon:Widget Switcher Object Reference), Index=1)
   > SetVisibility(self=Get HorizontalBox_Fruit() -> (HorizontalBox_Fruit:Horizontal Box Object Reference), InVisibility=Collapsed)
   > UpdateColor!(execute, self, Color:Linear Color Structure=(R=1,G=0,B=0,A=1)) -> (then)
  ]}
 EventGraph {nodes [
   OnEnemyHarvested() {is=title K2Node_CustomEvent}
   > EnableEnemyEventState!(execute, self, IsEnemyEventActive:bool=true, StatusIndex:int=3) -> (then) {is=title K2Node_CallFunction}
   n1 = OnInputMethodChanged!() -> (OutputDelegate, then, bNewInputType:ECommonInputType Enum) {is=title K2Node_CustomEvent}
   > UpdateKeybind()
   OnEnemyDamaged() {is=title K2Node_CustomEvent}
   > EnableEnemyEventState(IsEnemyEventActive=true, StatusIndex=4)
   n2 = Event SetMarkerData!() -> (OutputDelegate, then, MarkerData:"Fort World Marker Data Structure (by ref)", DisplayInfo:"Marked Actor Display Info Structure (by ref)") {is=title K2Node_Event}
   > Set CachedDisplayInfo(CachedDisplayInfo:Marked Actor Display Info Structure=@n2.DisplayInfo) -> (Output_Get:Marked Actor Display Info Structure)
   > Set MyMarkerData(MyMarkerData:Fort World Marker Data Structure=@n2.MarkerData) -> (Output_Get:Fort World Marker Data Structure)
   > n3 = Sequence()
   n4 = Get MyMarkerData() -> (MyMarkerData:Fort World Marker Data Structure)
   > Set CachedRealitySapling(CachedRealitySapling:B SMA Athena Reality Sapling Parent Object Reference=Cast To B_SMA_Athena_RealitySapling_Parent!(Object=Break Fort World Marker Data!(FortWorldMarkerData:"Fort World Marker Data Structure (by ref)"=@n4) -> (
      MarkerID:Marker ID Structure
      Owner:Fort Player State Weak object
      MarkerType:EFortWorldMarkerType Enum
      BasePositionOffset:Vector=~
      WorldNormal:Vector=~
      ItemDefinition:Item Definition Base Object Reference
      ItemCount:int
      PinnedQuestIndex:int
      MarkedActor:Actor Weak object
      MarkedActorClass:Actor Soft Class Reference
      MarkerTag:Gameplay Tag Structure
      bUseHoveredMarkerDetail:bool=~
     ) {is=title K2Node_BreakStruct}.MarkedActor) -> (AsB SMA Athena Reality Sapling Parent:B SMA Athena Reality Sapling Parent Object Reference, bSuccess:bool=~) {is=title K2Node_DynamicCast}) -> (Output_Get:B SMA Athena Reality Sapling Parent Object Reference)
   > UpdateDisplayText!(
    execute
    self
    DefaultText:Text=Break Marked Actor Display Info!(MarkedActorDisplayInfo:"Marked Actor Display Info Structure (by ref)"=Get CachedDisplayInfo() -> (CachedDisplayInfo:Marked Actor Display Info Structure)) -> (
     DisplayName:Text
     Icon:Texture 2D Soft Object Reference
     CustomIndicatorClass:User Widget Soft Class Reference
     CustomDetailsClass:User Widget Soft Class Reference
     PrimaryColor:Linear Color Structure
     Sound:Sound Base Object Reference
     bAttachSound:bool=~
     ScreenClamping:EFortMarkedActorScreenClamping Enum
     SpecialCull:Bitmask Integer
     SpecialDotSizeThreshold:float=~
     SpecialDistance:float=~
     CustomTimeToLive:float=~
     AdditionalData:Marked Actor Display Info Additional Data Object Reference
    ) {is=title K2Node_BreakStruct}
   ) -> (then)
   n5 = Is Valid(exec=@n3.then_1, InputObject=Get CachedRealitySapling() -> (CachedRealitySapling:B SMA Athena Reality Sapling Parent Object Reference))
   UpdateColor(
    execute=@n5.Is Not Valid
    Color=Break Marked Actor Display Info(MarkedActorDisplayInfo=Get CachedDisplayInfo()) -> (bAttachSound=~, SpecialDotSizeThreshold=~, SpecialDistance=~, CustomTimeToLive=~).PrimaryColor
   )
   n6 = Get CachedRealitySapling()
   Bind Event to Sapling Harvested by Enemy Event(execute=@n5.Is Valid, self:B SMA Athena Reality Sapling Parent Object Reference=@n6, Delegate=Create Event()) {is=title K2Node_AddDelegate}
   > Bind Event to Sapling Damaged by Enemy Event(self:B SMA Athena Reality Sapling Parent Object Reference=@n6, Delegate=Create Event()) {is=title K2Node_AddDelegate}
   n7 = Get CachedRealitySapling()
   > UpdateStatus!(
    execute
    self
    NeedsWeeding:bool="integer > integer"(A=Get NumCurrentWeeds!(self:B SMA Athena Reality Sapling Parent Object Reference=@n7) -> (NumCurrentWeeds:int=0))
    IsDead:bool=Is Dead!(self:B SMA Athena Reality Sapling Parent Object Reference=@n7) -> (RetVal:bool)
   ) -> (then)
   n8 = Get CachedRealitySapling()
   n9 = Get ServerSpawnedFruitSlot!(self:B SMA Athena Reality Sapling Parent Object Reference=@n8) -> (ServerSpawnedFruitSlot:Array of Server Spawned Fruit Slot Structures)
   > UpdateFruit!(execute, self, FruitStates:Array of Server Spawned Fruit Slot Structures=@n9) -> (then)
   n10 = Get FruitStates() -> (FruitStates:Array of Bytes)
   > ConvertToFruitVariantByteArray!(execute, FruitSlotArray:Array of Server Spawned Fruit Slot Structures=@n9, OutByteArray:Array of Bytes=@n10) -> (then)
   > n11 = Get Reality Sapling Rarity Data!(
    execute
    FruitStates:Array of Bytes=@n10
    NurtureLevel:Byte="To Byte (Integer)"!(InInt:int=Get SaplingNurturedCountThisMatch!(self:B SMA Athena Reality Sapling Parent Object Reference=@n8) -> (SaplingNurturedCountThisMatch:int=0)) -> (ReturnValue:Byte=0)
   ) -> (then, IsFruitGrowing:bool, RarityName:Text, RarityColor:Linear Color Structure) {is=title K2Node_CallFunction}
   > UpdateColor(Color=@n11.RarityColor)
   n12 = Get AthenaMarkerDetailsParent() -> (AthenaMarkerDetailsParent:Athena Marker Details Object Reference)
   > Is Valid(InputObject=@n12)
   > n13 = Cast To RealitySaplingPingMarker!(
    execute
    Object=GetMarkerWidget!(self:Fort World Marker Object Reference=Get FortWorldMarker!(self:Athena Marker Details Object Reference=@n12) -> (FortWorldMarker:Fort World Marker Object Reference)) -> (ReturnValue:User Widget Object Reference)
   ) -> (then, CastFailed, AsReality Sapling Ping Marker:Reality Sapling Ping Marker Object Reference) {is=title K2Node_DynamicCast}
   n14 = GetEnemyEventState!(self:Reality Sapling Ping Marker Object Reference=@n13) -> (EnemyEventState:E_SaplingEnemyEventState Enum=NewEnumerator0)
   > EnableEnemyEventState(
    IsEnemyEventActive="Not Equal (Enum)"!(A:E_SaplingEnemyEventState Enum=@n14, B:E_SaplingEnemyEventState Enum=NewEnumerator0) -> (ReturnValue:bool)
    StatusIndex=Select!(NewEnumerator0:int=0, NewEnumerator1:int=4, NewEnumerator2:int=3, Index:E_SaplingEnemyEventState Enum=@n14) -> (ReturnValue:int)
   )
   Event Construct()
   > "SetText (Text)"(self=Get ActionText() -> (ActionText:Common Text Object Reference), InText="LOCTEXT(\"Cancel Marker\")")
   > UpdateKeybind()
   > Bind Event to On Input Method Changed(self=Get CommonInputSubsystem(), Delegate=@n1)
   > n15 = FindParentWidgetOfType!(
    execute
    StartingWidget:Widget Object Reference=GetParent!(self:Widget Object Reference) -> (ReturnValue:Panel Widget Object Reference)
    Type:Widget Class Reference=/Game/Athena/HUD/Team/AthenaMarkerDetails.AthenaMarkerDetails_C
   ) -> (then, ReturnValue:Athena Marker Details Object Reference)
   > n16 = Cast To AthenaMarkerDetails!(execute, Object=@n15) -> (then, CastFailed, AsAthena Marker Details:Athena Marker Details Object Reference) {is=title K2Node_DynamicCast}
   > Set AthenaMarkerDetailsParent(AthenaMarkerDetailsParent:Athena Marker Details Object Reference=@n16) -> (Output_Get:Athena Marker Details Object Reference)
  ]}
 OnTakeAnyDamage {nodes [
   n1 = OnTakeAnyDamage!() -> (
    then
    DamagedActor:Actor Object Reference
    Damage:double=~
    DamageType:Damage Type Object Reference
    InstigatedBy:Controller Object Reference
    DamageCauser:Actor Object Reference
   ) {is=title K2Node_FunctionEntry}
   > n2 = Cast To FortPlayerController(Object=@n1.InstigatedBy)
   > Set AttackerPlayerController(AttackerPlayerController:Fort Player Controller Object Reference=@n2) -> (Output_Get:Fort Player Controller Object Reference)
   > Is Valid(InputObject=Get CachedRealitySapling())
   n3 = Get CachedRealitySapling()
   > Branch(Condition=OnSameTeam!(
     ActorA:Actor Object Reference=Get AttackerPlayerController() -> (AttackerPlayerController:Fort Player Controller Object Reference)
     ActorB:Actor Object Reference=Get Owning Player Controller Repped!(self:B SMA Athena Reality Sapling Parent Object Reference=@n3) -> (Owning Player Controller Repped:Fort Player Controller Object Reference)
    ) -> (ReturnValue:bool))
   > EnableEnemyEventState(IsEnemyEventActive=true, StatusIndex=5)
  ]}
 UpdateColor {nodes [
   n1 = UpdateColor!() -> (then, Color:Linear Color Structure) {is=title K2Node_FunctionEntry}
   > SetBrushColor(self=Get Seperator() -> (Seperator:Border Object Reference), InBrushColor=@n1)
   > UpdateIcon!(
    execute
    self
    Icon:Texture 2D Soft Object Reference=Break Marked Actor Display Info(MarkedActorDisplayInfo=Get CachedDisplayInfo()) -> (bAttachSound=~, SpecialDotSizeThreshold=~, SpecialDistance=~, CustomTimeToLive=~).Icon
    Color:Linear Color Structure=@n1
   ) -> (then)
  ]}
 UpdateDisplayText {nodes [
   n1 = UpdateDisplayText!() -> (then, DefaultText:Text) {is=title K2Node_FunctionEntry}
   > SetVisibility(self=Get Text_PlayerName() -> (Text_PlayerName:Common Text Object Reference), InVisibility=Visible)
   n2 = Get CachedRealitySapling()
   > n3 = Is Valid(InputObject=@n2)
   n4 = Get Owning Player State Repped!(self:B SMA Athena Reality Sapling Parent Object Reference=@n2) -> (Owning Player State Repped:Fort Player State Athena Object Reference)
   > n5 = Is Valid(InputObject=@n4)
   > n6 = Cast To FortPlayerControllerAthena(Object=GetOwningPlayer())
   n7 = GetSettingsFilteredPlayerName!(self:Fort Player Controller Gameplay Object Reference=@n6, OtherPlayerState:Fort Player State Object Reference=@n4) -> (ReturnValue:String)
   > "SetText (Text)"(
    self=Get Text_PlayerName()
    InText=Format Text!(Format="LOCTEXT(\"{username} sapling\")", username:Text="To Text (String)"(InString=@n7)) -> (Result)
   )
   n8 = GetOwningPlayer()
   n9 = Get Owning Player State Repped(self=Get CachedRealitySapling())
   > SetVisibility(
    self=Get Image_PingMarker() -> (Image_PingMarker:Image Object Reference)
    InVisibility=Select!(Option 0:ESlateVisibility Enum=Collapsed, Option 1:ESlateVisibility Enum=SelfHitTestInvisible, Index:bool=OnSameTeam(ActorA=@n8, ActorB=@n9)) -> (ReturnValue:ESlateVisibility Enum)
   )
   n10 = Get CachedRealitySapling()
   > SetColorAndOpacity(
    self:Image Object Reference=Get Image_PingMarker()
    InColorAndOpacity:Linear Color Structure=GetPinColor!(self:Fort Player State Zone Object Reference=Get Owning Player State Repped(self=@n10)) -> (ReturnValue:Linear Color Structure)
   )
   n11 = GetPlayerController!(self:Player State Object Reference=Get Owning Player State Repped(self=Get CachedRealitySapling())) -> (ReturnValue:Player Controller Object Reference)
   n12 = GetOwningPlayer()
   > Branch(Condition=AND Boolean(A="Equal (Object)"(A=@n12, B=@n11), B=IsValid(Object=@n11)))
   > "SetText (Text)"(self=Get Text_PlayerName(), InText="LOCTEXT(\"Your sapling\")")
   "SetText (Text)"(
    execute=[@n6.CastFailed, @n5.Is Not Valid, @n3.Is Not Valid]
    self=Get Text_PlayerName()
    InText=TextToUpper!(InText:"Text (by ref)"=@n1) -> (ReturnValue:Text)
   )
   > SetVisibility(self=Get Image_PingMarker(), InVisibility=Collapsed)
  ]}
 UpdateFruit {nodes [
   n1 = UpdateFruit!() -> (then, FruitStates:Array of Server Spawned Fruit Slot Structures) {is=title K2Node_FunctionEntry}
   > Set AvailableFruitCount(AvailableFruitCount:int=0) -> (Output_Get:int=0)
   > Set RarityColor(RarityColor:Linear Color Structure=Break Marked Actor Display Info(MarkedActorDisplayInfo=Get CachedDisplayInfo()) -> (bAttachSound=~, SpecialDotSizeThreshold=~, SpecialDistance=~, CustomTimeToLive=~).PrimaryColor) -> (Output_Get:Linear Color Structure)
   > n2 = For Each Loop!(Exec, Array:Array of Server Spawned Fruit Slot Structures=@n1) -> (
    LoopBody
    Array Element_Slot_4_#1:int
    Array Element_Fruit_5_#2:Tiered Reality Fruit Athena Object Reference
    Array Index
    Completed
   )
   n3 = Get Is Enemy Event Active() -> (Is Enemy Event Active:bool)
   n4 = Get AvailableFruitCount() -> (AvailableFruitCount:int=0)
   n5 = Branch(execute=@n2.Completed, Condition=AND Boolean(A="integer > integer"(A=@n4), B=NOT Boolean(A=@n3)))
   n6 = Get HorizontalBox_Fruit()
   > SetVisibility(self=@n6, InVisibility=SelfHitTestInvisible)
   > SetRarityColor!(
    execute
    self:Reality Fruit Object Reference=Get RealityFruit() -> (RealityFruit:Reality Fruit Object Reference)
    Color:Linear Color Structure=Get RarityColor() -> (RarityColor:Linear Color Structure)
   ) -> (then)
   > SetReady!(execute, self:Reality Fruit Object Reference=Get RealityFruit(), Ready:bool=true) -> (then)
   > n7 = Branch(Condition="integer > integer"(A=Get AvailableFruitCount() -> (AvailableFruitCount=0), B=1))
   n8 = Get Text_FruitQuantity() -> (Text_FruitQuantity:Common Text Object Reference)
   > SetVisibility(self=@n8, InVisibility=SelfHitTestInvisible)
   > "SetText (Text)"(
    self=Get Text_FruitQuantity()
    InText=Format Text!(Format="LOCTEXT(\"x{FruitCount}\")", FruitCount:int=Get AvailableFruitCount() -> (AvailableFruitCount=0)) -> (Result)
   )
   SetVisibility(execute=@n7.else, self=@n8, InVisibility=Collapsed)
   SetVisibility(execute=@n5.else, self=@n6, InVisibility=Collapsed)
   n9 = Check Sapling Fruit Status!(
    execute:Exec=@n2.LoopBody
    FruitVariant:SaplingFruitVariants Enum=Get VariantOrRarity!(self:Tiered Reality Fruit Athena Object Reference=@n2.Array Element_Fruit_5_#2) -> (VariantOrRarity:SaplingFruitVariants Enum=NewEnumerator0)
   ) -> (NoFruit:Exec, IsGrowing:Exec, IsReady:Exec, RarityColor:Linear Color Structure) {is=title K2Node_MacroInstance}
   Increment Int(" "=@n9.IsReady, Value=Get AvailableFruitCount() -> (AvailableFruitCount=0))
   > Set RarityColor(RarityColor=@n9)
  ]}
 UpdateIcon {nodes [
   n1 = UpdateIcon!() -> (then, Icon:Texture 2D Soft Object Reference, Color:Linear Color Structure) {is=title K2Node_FunctionEntry}
   > n2 = GetDynamicMaterial(self=Get ItemIcon() -> (ItemIcon:Image Object Reference))
   > SetScalarParameterValue(self=@n2, ParameterName=UseIconColor, Value=1)
   > SetScalarParameterValue(self=@n2, ParameterName=IconColorStrength, Value=0.94)
   > SetTextureParameterValue(
    self=@n2
    ParameterName=DisplayedIcon
    Value=Resolve Soft Reference!(Input:Texture 2D Soft Object Reference=@n1) -> (Output:Texture 2D Object Reference)
   )
   n3 = BreakColor(InColor=@n1.Color)
   n4 = Get ColorReduction() -> (ColorReduction:double)
   > SetVectorParameterValue(
    self=@n2
    ParameterName=FillColor
    Value=MakeColor!(R:float=float * float(A=@n3, B=@n4), G:float=float * float(A=@n3.G, B=@n4), B:float=float * float(A=@n3.B, B=@n4), A:float=0.9) -> (ReturnValue:Linear Color Structure)
   )
   > SetVectorParameterValue(self=@n2, ParameterName=FillColorInner, Value=@n1.Color)
   > SetVectorParameterValue(self=@n2, ParameterName=RingColor, Value=@n1.Color)
   > SetVectorParameterValue(self=@n2, ParameterName=IconColor, Value=(R=1,G=1,B=1,A=1))
   > SetVectorParameterValue(self=@n2, ParameterName=IconScale, Value=MakeColor(R=0.85, G=0.85, B=0, A=1))
   n5 = Get AthenaMarkerDetailsParent()
   > Is Valid(InputObject=@n5)
   > SetLineColor!(execute, self:Athena Marker Details Object Reference=@n5, NewColor:Linear Color Structure=@n1.Color) -> (then)
  ]}
 UpdateKeybind {nodes [
   UpdateKeybind() {is=title K2Node_FunctionEntry}
   n1 = Get CommonInputSubsystem()
   > SetBoundAction!(
    execute
    self:Fort Keybind Widget Object Reference=Get KeybindWidget_Map() -> (KeybindWidget_Map:Keybind Widget Object Reference)
    NewBoundAction:Name=Select!(
     MouseAndKeyboard:Name=ToggleFullScreenMap
     Gamepad:Name=GamepadToggleFullScreenMap
     Touch:Name=None
     Count:Name=None
     Index:ECommonInputType Enum=GetCurrentInputType(self=@n1)
    ) -> (ReturnValue:Name)
   ) -> (then)
  ]}
 UpdateStatus {nodes [
   n1 = UpdateStatus!() -> (then, NeedsWeeding:bool=~, IsDead:bool=~) {is=title K2Node_FunctionEntry}
   > SetActiveWidgetIndex(self=Get WidgetSwitcher_Status(), Index=SelectInt(A=2, B=SelectInt(A=1, B=0, bPickA=@n1), bPickA=@n1.IsDead))
  ]}
]
variables [
 AthenaMarkerDetailsParent {type=AthenaMarkerDetails_C, default=null}
 CachedDisplayInfo {
  type=MarkedActorDisplayInfo
  default {
   additionalData=null
   bAttachSound=true
   customDetailsClass=null
   customIndicatorClass=null
   customTimeToLive=-1
   displayName=""
   icon=null
   screenClamping=Default
   sound=null
   specialCull=0
   specialDistance=50
   specialDotSizeThreshold=-0.93
   primaryColor=<rgba 0 0 0 0>
  }
 }
 CachedRealitySapling {type=B_SMA_Athena_RealitySapling_Parent_C, default=null}
 ColorReduction {type=number, default=0.08}
 FruitStates {type="array<integer>"}
 Is Enemy Event Active {type=boolean}
 MyMarkerData {
  type=FortWorldMarkerData
  default {
   bHeldMarkerInput=false
   bIncludeSquad=true
   bUseHoveredMarkerDetail=true
   itemCount=1
   itemDefinition=null
   markedActor=null
   markedActorClass=null
   markerType=None
   mostRecentArrayReplicationKey=-1
   owner=null
   pingCommand=Invalid
   pinnedQuestIndex=-1
   replicationId=-1
   replicationKey=-1
   targetConvertedType=Unknown
   basePosition=<0 0 0>
   basePositionOffset=<0 0 0>
   confirmationContainer {confirmations []}
   customDisplayInfos []
   markerId {instanceId=-1, playerId=-1}
   markerTag {tagName=None}
   worldNormal=<0 0 1>
  }
 }
]
widgets {
 info {inheritedWidgetCount=0, parentClass=&UserWidget, rootWidgetClass=&Overlay, widgetCount=43}
 widgets [
  Overlay_16 {widgetClassPath=&Overlay, widget=Overlay_16}
  VerticalBox_0 {
   widgetClassPath=&VerticalBox
   parent=Overlay_16
   widget=VerticalBox_0
   slot {class=OverlaySlot, properties {horizontalAlignment=HAlign_Fill, verticalAlignment=VAlign_Fill}}
  }
  HorizontalBox_0 {
   widgetClassPath=&HorizontalBox
   parent=VerticalBox_0
   widget=HorizontalBox_0
   slot {class=VerticalBoxSlot, properties {horizontalAlignment=HAlign_Left, padding=<ltrb 6 0 8 -6>}}
  }
  Image_PingMarker {
   widgetClassPath=&Image
   bIsVariable=true
   parent=HorizontalBox_0
   widget=Image_PingMarker
   properties {brush {
     bIsDynamicallyLoaded=false
     drawAs=Image
     imageType=NoImage
     mirroring=NoMirror
     resourceName=None
     resourceObject=/Game/Athena/HUD/Art/T_UI_Indicator_Diamond
     tiling=NoTile
     imageSize=<18 22>
     margin=<ltrb 0 0 0 0>
     outlineSettings {
      bUseBrushTransparency=false
      roundingType=HalfHeightRadius
      width=0
      color {colorUseRule=UseColor_Specified, specifiedColor=<rgba 0 0 0 0>}
      cornerRadii=<0 0 0 0>
     }
     tintColor {colorUseRule=UseColor_Specified, specifiedColor=<rgba 1 1 1 1>}
     uVRegion {bIsValid=false, max=<0 0>, min=<0 0>}
    }}
   slot {class=HorizontalBoxSlot, properties {horizontalAlignment=HAlign_Left, verticalAlignment=VAlign_Center, padding=<ltrb 0 0 2 0>}}
  }
  Text_PlayerName {
   widgetClassPath=&CommonTextBlock
   bIsVariable=true
   parent=HorizontalBox_0
   widget=Text_PlayerName
   properties {
    mobileFontSizeMultiplier=1.4
    text=Unknown
    textTransformPolicy=ToUpper
    font {
     bForceMonospaced=false
     bMaterialIsStencil=false
     fontMaterial=null
     fontObject=&Burbank
     letterSpacing=0
     monospacedWidth=1
     size=16
     skewAmount=0
     typefaceFontName=BurbankRegular-Black
     outlineSettings {
      bApplyOutlineToDropShadows=false
      bMiteredCorners=false
      bSeparateFillAlpha=false
      outlineMaterial=&M-FontShadow
      outlineSize=3
      outlineColor=<rgba 0 0 0 1>
     }
    }
    shadowOffset=<0 0>
   }
   slot {class=HorizontalBoxSlot, properties {horizontalAlignment=HAlign_Left, verticalAlignment=VAlign_Center}}
  }
  Overlay_0 {
   widgetClassPath=&Overlay
   parent=VerticalBox_0
   widget=Overlay_0
   slot {class=VerticalBoxSlot, properties {padding=<ltrb 0 4 0 0>}}
  }
  Seperator {
   widgetClassPath=&Border
   bIsVariable=true
   parent=Overlay_0
   widget=Seperator
   properties {
    background {
     bIsDynamicallyLoaded=false
     drawAs=Image
     imageType=NoImage
     mirroring=NoMirror
     resourceName=None
     resourceObject=null
     tiling=NoTile
     imageSize=<32 16>
     margin=<ltrb 0 0 0 0>
     outlineSettings {
      bUseBrushTransparency=false
      roundingType=HalfHeightRadius
      width=0
      color {colorUseRule=UseColor_Specified, specifiedColor=<rgba 0 0 0 0>}
      cornerRadii=<0 0 0 0>
     }
     tintColor {colorUseRule=UseColor_Specified, specifiedColor=<rgba 1 1 1 1>}
     uVRegion {bIsValid=false, max=<0 0>, min=<0 0>}
    }
    brushColor=<rgba 1 0 0 1>
    padding=<ltrb 4 0 4 3>
    renderTransformPivot=<0 0.5>
   }
   slot {class=OverlaySlot, properties {horizontalAlignment=HAlign_Fill, verticalAlignment=VAlign_Center, padding=<ltrb 3 0 0 0>}}
  }
  FortVisualAttachment_0 {
   widgetClassPath=&FortVisualAttachment
   parent=Overlay_0
   widget=FortVisualAttachment_0
   properties {widthOverride=100}
   slot {class=OverlaySlot, properties {verticalAlignment=VAlign_Center}}
  }
  Overlay_1 {
   widgetClassPath=&Overlay
   parent=FortVisualAttachment_0
   widget=Overlay_1
   slot {class=SizeBoxSlot, properties {horizontalAlignment=HAlign_Right, verticalAlignment=VAlign_Top}}
  }
  Switcher_IndicatorIcon {
   widgetClassPath=&WidgetSwitcher
   bIsVariable=true
   parent=Overlay_1
   widget=Switcher_IndicatorIcon
   slot {class=OverlaySlot, properties {horizontalAlignment=HAlign_Fill, verticalAlignment=VAlign_Fill}}
  }
  ItemIcon {
   widgetClassPath=&Image
   bIsVariable=true
   parent=Switcher_IndicatorIcon
   widget=ItemIcon
   properties {
    brush {
     bIsDynamicallyLoaded=false
     drawAs=Image
     imageType=NoImage
     mirroring=NoMirror
     resourceName=None
     resourceObject=/Game/Athena/HUD/Materials/M_UI_PingBubble_LargeIItemPing
     tiling=NoTile
     imageSize=<75 75>
     margin=<ltrb 0 0 0 0>
     outlineSettings {
      bUseBrushTransparency=false
      roundingType=HalfHeightRadius
      width=0
      color {colorUseRule=UseColor_Specified, specifiedColor=<rgba 0 0 0 0>}
      cornerRadii=<0 0 0 0>
     }
     tintColor {colorUseRule=UseColor_Specified, specifiedColor=<rgba 1 1 1 1>}
     uVRegion {bIsValid=false, max=<0 0>, min=<0 0>}
    }
    renderTransform {angle=0, scale=<1 1>, shear=<0 0>, translation=<-65 -37>}
    renderTransformPivot=<1 0.5>
   }
   slot {class=WidgetSwitcherSlot}
  }
  DangerIcon {
   widgetClassPath=&Image
   bIsVariable=true
   parent=Switcher_IndicatorIcon
   widget=DangerIcon
   properties {
    brush {
     bIsDynamicallyLoaded=false
     drawAs=Image
     imageType=NoImage
     mirroring=NoMirror
     resourceName=None
     resourceObject=/Game/Athena/HUD/Materials/M_UI_PingBubble_Inst
     tiling=NoTile
     imageSize=<75 75>
     margin=<ltrb 0 0 0 0>
     outlineSettings {
      bUseBrushTransparency=false
      roundingType=HalfHeightRadius
      width=0
      color {colorUseRule=UseColor_Specified, specifiedColor=<rgba 0 0 0 0>}
      cornerRadii=<0 0 0 0>
     }
     tintColor {colorUseRule=UseColor_Specified, specifiedColor=<rgba 1 1 1 1>}
     uVRegion {bIsValid=false, max=<0 0>, min=<0 0>}
    }
    renderTransform {angle=0, scale=<1 1>, shear=<0 0>, translation=<-65 -37>}
    renderTransformPivot=<1 0.5>
   }
   slot {class=WidgetSwitcherSlot}
  }
  ScaleBox_1 {
   widgetClassPath=&ScaleBox
   parent=Overlay_1
   widget=ScaleBox_1
   properties {stretch=UserSpecified, userSpecifiedScale=0.6, renderTransform {angle=0, scale=<1 1>, shear=<0 0>, translation=<-35 5>}}
   slot {class=OverlaySlot, properties {horizontalAlignment=HAlign_Center, verticalAlignment=VAlign_Center, padding=<ltrb -55 -15 0 0>}}
  }
  KeybindWidget_Map {
   widgetClassPath=/FortUILibrary/Input/KeybindWidget
   bIsVariable=true
   parent=ScaleBox_1
   widget=KeybindWidget_Map
   properties {boundAction=ToggleFullScreenMap}
   slot {class=ScaleBoxSlot}
  }
  VerticalBox_3 {
   widgetClassPath=&VerticalBox
   parent=VerticalBox_0
   widget=VerticalBox_3
   slot {class=VerticalBoxSlot, properties {horizontalAlignment=HAlign_Left, padding=<ltrb 4 4 8 0>}}
  }
  HorizontalBox_1 {
   widgetClassPath=&HorizontalBox
   parent=VerticalBox_3
   widget=HorizontalBox_1
   slot {class=VerticalBoxSlot, properties {padding=<ltrb 16 0 0 0>}}
  }
  HorizontalBox_Fruit {
   widgetClassPath=&HorizontalBox
   bIsVariable=true
   parent=HorizontalBox_1
   widget=HorizontalBox_Fruit
   slot {class=HorizontalBoxSlot, properties {padding=<ltrb 0 0 10 0>}}
  }
  RealityFruit {
   widgetClassPath=&/RealityFruit
   bIsVariable=true
   parent=HorizontalBox_Fruit
   widget=RealityFruit
   properties {useShadow=true}
   slot {class=HorizontalBoxSlot}
  }
  Text_FruitQuantity {
   widgetClassPath=&CommonTextBlock
   bIsVariable=true
   parent=HorizontalBox_Fruit
   widget=Text_FruitQuantity
   properties {
    mobileFontSizeMultiplier=1.4
    text=x3
    font {
     bForceMonospaced=false
     bMaterialIsStencil=false
     fontMaterial=null
     fontObject=&BurbankSmall
     letterSpacing=0
     monospacedWidth=1
     size=14
     skewAmount=0
     typefaceFontName=Bold
     outlineSettings {
      bApplyOutlineToDropShadows=false
      bMiteredCorners=false
      bSeparateFillAlpha=false
      outlineMaterial=&M-FontShadow
      outlineSize=3
      outlineColor=<rgba 0 0 0 1>
     }
    }
    renderTransform {angle=0, scale=<1 1>, shear=<-12 0>, translation=<0 0>}
    shadowOffset=<0 0>
   }
   slot {class=HorizontalBoxSlot, properties {verticalAlignment=VAlign_Center, padding=<ltrb -5 0 0 0>}}
  }
  WidgetSwitcher_Status {
   widgetClassPath=&WidgetSwitcher
   bIsVariable=true
   parent=HorizontalBox_1
   widget=WidgetSwitcher_Status
   properties {activeWidgetIndex=1, renderTransform {angle=0, scale=<1 1>, shear=<-12 0>, translation=<0 0>}}
   slot {class=HorizontalBoxSlot}
  }
  HorizontalBox_Empty {widgetClassPath=&HorizontalBox, parent=WidgetSwitcher_Status, widget=HorizontalBox_Empty, slot {class=WidgetSwitcherSlot}}
  HorizontalBox_Weeding {
   widgetClassPath=&HorizontalBox
   parent=WidgetSwitcher_Status
   widget=HorizontalBox_Weeding
   properties {flowDirectionPreference=Culture}
   slot {class=WidgetSwitcherSlot}
  }
  Image_Weeding {
   widgetClassPath=&Image
   parent=HorizontalBox_Weeding
   widget=Image_Weeding
   properties {brush {
     bIsDynamicallyLoaded=false
     drawAs=Image
     imageType=NoImage
     mirroring=NoMirror
     resourceName=None
     resourceObject=/RealitySeedGameplay/UI/Icons/M_UI_NeedsWeeds_1
     tiling=NoTile
     imageSize=<32 32>
     margin=<ltrb 0 0 0 0>
     outlineSettings {
      bUseBrushTransparency=false
      roundingType=HalfHeightRadius
      width=0
      color {colorUseRule=UseColor_Specified, specifiedColor=<rgba 0 0 0 0>}
      cornerRadii=<0 0 0 0>
     }
     tintColor {colorUseRule=UseColor_Specified, specifiedColor=<rgba 1 1 1 1>}
     uVRegion {bIsValid=false, max=<0 0>, min=<0 0>}
    }}
   slot {class=HorizontalBoxSlot, properties {horizontalAlignment=HAlign_Center, verticalAlignment=VAlign_Center}}
  }
  ActionText_NeedsWeeding {
   widgetClassPath=&CommonTextBlock
   parent=HorizontalBox_Weeding
   widget=ActionText_NeedsWeeding
   properties {
    mobileFontSizeMultiplier=1.4
    text=Weeds
    textTransformPolicy=ToUpper
    font {
     bForceMonospaced=false
     bMaterialIsStencil=false
     fontMaterial=null
     fontObject=&BurbankSmall
     letterSpacing=0
     monospacedWidth=1
     size=14
     skewAmount=0
     typefaceFontName=Bold
     outlineSettings {
      bApplyOutlineToDropShadows=false
      bMiteredCorners=false
      bSeparateFillAlpha=false
      outlineMaterial=&M-FontShadow
      outlineSize=3
      outlineColor=<rgba 0 0 0 1>
     }
    }
    shadowOffset=<0 0>
   }
   slot {class=HorizontalBoxSlot, properties {verticalAlignment=VAlign_Center, padding=<ltrb -2 0 0 0>}}
  }
  Yes_Indicator {
   widgetClassPath=&Overlay
   parent=HorizontalBox_Weeding
   widget=Yes_Indicator
   properties {visibility=Collapsed, renderTransform {angle=0, scale=<1 1>, shear=<0 0>, translation=<0 -1>}}
   slot {class=HorizontalBoxSlot, properties {horizontalAlignment=HAlign_Center, verticalAlignment=VAlign_Center}}
  }
  Image_Check_Shadow {
   widgetClassPath=&Image
   parent=Yes_Indicator
   widget=Image_Check_Shadow
   properties {
    renderOpacity=0.25
    brush {
     bIsDynamicallyLoaded=false
     drawAs=Image
     imageType=NoImage
     mirroring=NoMirror
     resourceName=None
     resourceObject=/Game/UI/Foundation/Textures/Icons/Checks/T-Icon-Check-64
     tiling=NoTile
     imageSize=<16 16>
     margin=<ltrb 0 0 0 0>
     outlineSettings {
      bUseBrushTransparency=false
      roundingType=HalfHeightRadius
      width=0
      color {colorUseRule=UseColor_Specified, specifiedColor=<rgba 0 0 0 0>}
      cornerRadii=<0 0 0 0>
     }
     tintColor {colorUseRule=UseColor_Specified, specifiedColor=<rgba 1 1 1 1>}
     uVRegion {bIsValid=false, max=<0 0>, min=<0 0>}
    }
    colorAndOpacity=<rgba 0 0 0 1>
    renderTransform {angle=0, scale=<1 1>, shear=<0 0>, translation=<1.5 1.5>}
   }
   slot {class=OverlaySlot, properties {verticalAlignment=VAlign_Center}}
  }
  Image_Check {
   widgetClassPath=&Image
   parent=Yes_Indicator
   widget=Image_Check
   properties {
    brush {
     bIsDynamicallyLoaded=false
     drawAs=Image
     imageType=NoImage
     mirroring=NoMirror
     resourceName=None
     resourceObject=&T-Icon-Check-64
     tiling=NoTile
     imageSize=<16 16>
     margin=<ltrb 0 0 0 0>
     outlineSettings {
      bUseBrushTransparency=false
      roundingType=HalfHeightRadius
      width=0
      color {colorUseRule=UseColor_Specified, specifiedColor=<rgba 0 0 0 0>}
      cornerRadii=<0 0 0 0>
     }
     tintColor {colorUseRule=UseColor_Specified, specifiedColor=<rgba 1 1 1 1>}
     uVRegion {bIsValid=false, max=<0 0>, min=<0 0>}
    }
    colorAndOpacity=<rgba 0.1 1 0.4 1>
   }
   slot {class=OverlaySlot, properties {verticalAlignment=VAlign_Center}}
  }
  HorizontalBox_Dead {
   widgetClassPath=&HorizontalBox
   parent=WidgetSwitcher_Status
   widget=HorizontalBox_Dead
   properties {flowDirectionPreference=Culture}
   slot {class=WidgetSwitcherSlot}
  }
  Image_Dead {
   widgetClassPath=&Image
   parent=HorizontalBox_Dead
   widget=Image_Dead
   properties {brush {
     bIsDynamicallyLoaded=false
     drawAs=Image
     imageType=NoImage
     mirroring=NoMirror
     resourceName=None
     resourceObject=/RealitySeedGameplay/UI/Icons/M_UI_SaplingDead
     tiling=NoTile
     imageSize=<32 32>
     margin=<ltrb 0 0 0 0>
     outlineSettings {
      bUseBrushTransparency=false
      roundingType=HalfHeightRadius
      width=0
      color {colorUseRule=UseColor_Specified, specifiedColor=<rgba 0 0 0 0>}
      cornerRadii=<0 0 0 0>
     }
     tintColor {colorUseRule=UseColor_Specified, specifiedColor=<rgba 1 1 1 1>}
     uVRegion {bIsValid=false, max=<0 0>, min=<0 0>}
    }}
   slot {class=HorizontalBoxSlot, properties {horizontalAlignment=HAlign_Center, verticalAlignment=VAlign_Center}}
  }
  ActionText_Dead {
   widgetClassPath=&CommonTextBlock
   parent=HorizontalBox_Dead
   widget=ActionText_Dead
   properties {
    mobileFontSizeMultiplier=1.4
    text=Destroyed
    textTransformPolicy=ToUpper
    font {
     bForceMonospaced=false
     bMaterialIsStencil=false
     fontMaterial=null
     fontObject=&BurbankSmall
     letterSpacing=0
     monospacedWidth=1
     size=14
     skewAmount=0
     typefaceFontName=Bold
     outlineSettings {
      bApplyOutlineToDropShadows=true
      bMiteredCorners=false
      bSeparateFillAlpha=true
      outlineMaterial=&M-FontShadow
      outlineSize=3
      outlineColor=<rgba 0 0 0 1>
     }
    }
    shadowOffset=<0 0>
   }
   slot {class=HorizontalBoxSlot, properties {verticalAlignment=VAlign_Center}}
  }
  X_Indicator {
   widgetClassPath=&Overlay
   parent=HorizontalBox_Dead
   widget=X_Indicator
   properties {visibility=Collapsed, renderTransform {angle=0, scale=<1 1>, shear=<0 0>, translation=<0 -1>}}
   slot {class=HorizontalBoxSlot, properties {horizontalAlignment=HAlign_Center, verticalAlignment=VAlign_Center}}
  }
  Image_X_Shadow {
   widgetClassPath=&Image
   parent=X_Indicator
   widget=Image_X_Shadow
   properties {
    renderOpacity=0.25
    brush {
     bIsDynamicallyLoaded=false
     drawAs=Image
     imageType=NoImage
     mirroring=NoMirror
     resourceName=None
     resourceObject=/Game/Athena/HUD/Art/squadArt/squad_Eliminated
     tiling=NoTile
     imageSize=<16 16>
     margin=<ltrb 0 0 0 0>
     outlineSettings {
      bUseBrushTransparency=false
      roundingType=HalfHeightRadius
      width=0
      color {colorUseRule=UseColor_Specified, specifiedColor=<rgba 0 0 0 0>}
      cornerRadii=<0 0 0 0>
     }
     tintColor {colorUseRule=UseColor_Specified, specifiedColor=<rgba 1 1 1 1>}
     uVRegion {bIsValid=false, max=<0 0>, min=<0 0>}
    }
    colorAndOpacity=<rgba 0 0 0 1>
    renderTransform {angle=0, scale=<1 1>, shear=<0 0>, translation=<1.5 1.5>}
   }
   slot {class=OverlaySlot}
  }
  Image_X {
   widgetClassPath=&Image
   parent=X_Indicator
   widget=Image_X
   properties {brush {
     bIsDynamicallyLoaded=false
     drawAs=Image
     imageType=NoImage
     mirroring=NoMirror
     resourceName=None
     resourceObject=&squad_Eliminated
     tiling=NoTile
     imageSize=<16 16>
     margin=<ltrb 0 0 0 0>
     outlineSettings {
      bUseBrushTransparency=false
      roundingType=HalfHeightRadius
      width=0
      color {colorUseRule=UseColor_Specified, specifiedColor=<rgba 0 0 0 0>}
      cornerRadii=<0 0 0 0>
     }
     tintColor {colorUseRule=UseColor_Specified, specifiedColor=<rgba 1 0.0375 0.0696 1>}
     uVRegion {bIsValid=false, max=<0 0>, min=<0 0>}
    }}
   slot {class=OverlaySlot}
  }
  HorizontalBox_EnemyLooted {
   widgetClassPath=&HorizontalBox
   parent=WidgetSwitcher_Status
   widget=HorizontalBox_EnemyLooted
   properties {flowDirectionPreference=Culture}
   slot {class=WidgetSwitcherSlot}
  }
  Image_EnemyLooting {
   widgetClassPath=&Image
   parent=HorizontalBox_EnemyLooted
   widget=Image_EnemyLooting
   properties {brush {
     bIsDynamicallyLoaded=false
     drawAs=Image
     imageType=NoImage
     mirroring=NoMirror
     resourceName=None
     resourceObject=/RealitySeedGameplay/UI/Icons/M_UI_EnemyLooting
     tiling=NoTile
     imageSize=<32 32>
     margin=<ltrb 0 0 0 0>
     outlineSettings {
      bUseBrushTransparency=false
      roundingType=HalfHeightRadius
      width=0
      color {colorUseRule=UseColor_Specified, specifiedColor=<rgba 0 0 0 0>}
      cornerRadii=<0 0 0 0>
     }
     tintColor {colorUseRule=UseColor_Specified, specifiedColor=<rgba 1 1 1 1>}
     uVRegion {bIsValid=false, max=<0 0>, min=<0 0>}
    }}
   slot {class=HorizontalBoxSlot, properties {horizontalAlignment=HAlign_Center, verticalAlignment=VAlign_Center}}
  }
  ActionText_EnemyLooting {
   widgetClassPath=&CommonTextBlock
   parent=HorizontalBox_EnemyLooted
   widget=ActionText_EnemyLooting
   properties {
    mobileFontSizeMultiplier=1.4
    text=Enemy Looting
    textTransformPolicy=ToUpper
    font {
     bForceMonospaced=false
     bMaterialIsStencil=false
     fontMaterial=null
     fontObject=&BurbankSmall
     letterSpacing=0
     monospacedWidth=1
     size=14
     skewAmount=0
     typefaceFontName=Bold
     outlineSettings {
      bApplyOutlineToDropShadows=true
      bMiteredCorners=false
      bSeparateFillAlpha=true
      outlineMaterial=&M-FontShadow
      outlineSize=3
      outlineColor=<rgba 0 0 0 1>
     }
    }
    shadowOffset=<0 0>
   }
   slot {class=HorizontalBoxSlot, properties {verticalAlignment=VAlign_Center}}
  }
  HorizontalBox_EnemyAttacking {
   widgetClassPath=&HorizontalBox
   parent=WidgetSwitcher_Status
   widget=HorizontalBox_EnemyAttacking
   properties {flowDirectionPreference=Culture}
   slot {class=WidgetSwitcherSlot}
  }
  Image_EnemyAttacking {
   widgetClassPath=&Image
   parent=HorizontalBox_EnemyAttacking
   widget=Image_EnemyAttacking
   properties {brush {
     bIsDynamicallyLoaded=false
     drawAs=Image
     imageType=NoImage
     mirroring=NoMirror
     resourceName=None
     resourceObject=/RealitySeedGameplay/UI/Icons/M_UI_EnemyAttacking
     tiling=NoTile
     imageSize=<32 32>
     margin=<ltrb 0 0 0 0>
     outlineSettings {
      bUseBrushTransparency=false
      roundingType=HalfHeightRadius
      width=0
      color {colorUseRule=UseColor_Specified, specifiedColor=<rgba 0 0 0 0>}
      cornerRadii=<0 0 0 0>
     }
     tintColor {colorUseRule=UseColor_Specified, specifiedColor=<rgba 1 1 1 1>}
     uVRegion {bIsValid=false, max=<0 0>, min=<0 0>}
    }}
   slot {class=HorizontalBoxSlot, properties {horizontalAlignment=HAlign_Center, verticalAlignment=VAlign_Center}}
  }
  ActionText_EnemyAttacking {
   widgetClassPath=&CommonTextBlock
   parent=HorizontalBox_EnemyAttacking
   widget=ActionText_EnemyAttacking
   properties {
    mobileFontSizeMultiplier=1.4
    text=Under Attack
    textTransformPolicy=ToUpper
    font {
     bForceMonospaced=false
     bMaterialIsStencil=false
     fontMaterial=null
     fontObject=&BurbankSmall
     letterSpacing=0
     monospacedWidth=1
     size=14
     skewAmount=0
     typefaceFontName=Bold
     outlineSettings {
      bApplyOutlineToDropShadows=true
      bMiteredCorners=false
      bSeparateFillAlpha=true
      outlineMaterial=&M-FontShadow
      outlineSize=3
      outlineColor=<rgba 0 0 0 1>
     }
    }
    shadowOffset=<0 0>
   }
   slot {class=HorizontalBoxSlot, properties {verticalAlignment=VAlign_Center}}
  }
  ActionTextContainer {
   widgetClassPath=&HorizontalBox
   parent=VerticalBox_3
   widget=ActionTextContainer
   slot {class=VerticalBoxSlot, properties {padding=<ltrb 0 8 0 0>}}
  }
  ScaleBox_0 {
   widgetClassPath=&ScaleBox
   parent=ActionTextContainer
   widget=ScaleBox_0
   properties {stretch=UserSpecified, userSpecifiedScale=0.6}
   slot {class=HorizontalBoxSlot, properties {verticalAlignment=VAlign_Center}}
  }
  KeybindWidget {
   widgetClassPath=&KeybindWidget
   parent=ScaleBox_0
   widget=KeybindWidget
   properties {boundAction=PlaceMapMarkerInWorld}
   slot {class=ScaleBoxSlot}
  }
  ActionText {
   widgetClassPath=&CommonTextBlock
   bIsVariable=true
   parent=ActionTextContainer
   widget=ActionText
   properties {
    mobileFontSizeMultiplier=1.4
    text=Text Block
    textTransformPolicy=ToUpper
    font {
     bForceMonospaced=false
     bMaterialIsStencil=false
     fontMaterial=null
     fontObject=&BurbankSmall
     letterSpacing=0
     monospacedWidth=1
     size=14
     skewAmount=0
     typefaceFontName=Black
     outlineSettings {
      bApplyOutlineToDropShadows=true
      bMiteredCorners=false
      bSeparateFillAlpha=true
      outlineMaterial=&M-FontShadow
      outlineSize=3
      outlineColor=<rgba 0 0 0 1>
     }
    }
    shadowOffset=<0 0>
   }
   slot {class=HorizontalBoxSlot, properties {verticalAlignment=VAlign_Center, padding=<ltrb 4 -2 0 -2>}}
  }
 ]
}
