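# Squad indicator widget that serves as the ping marker for a Reality Sapling, showing its rarity or an enemy threat.
# On marker data it casts the marked actor to a sapling, rebinds update, harvest and damage delegates, and clears threat state.
# UpdateColor tints indicators by rarity derived from fruit states and nurture count.
# Enemy attack or harvest switches to a red danger icon, which sapling updates do not clear.
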WidgetBlueprint /RealitySeedGameplay/UI/RealitySaplingPingMarker
animations [
 Anim_DangerPlaced {
  binding_count=1
  length_seconds=0.651
  object=&self:Anim_DangerPlaced
  sequence {bindings [Burst {
     id=#1
     possessed_class=Image
     tracks [
      {
       type=MovieSceneWidgetMaterialTrack
       sections [{
         type=MovieSceneParameterSection
         end_frame=8
         start_frame=-5
         channels [
          Alpha {type=MovieSceneScriptingFloatChannel, thinned_from=7, keys [[-0.949,0],[-0.5157,0],[0.4843,1],[3.484,1],[6,0.1967],[7.152,0]]}
          Grow {type=MovieSceneScriptingFloatChannel, thinned_from=4, keys [[-0.949,0],[-0.5157,0],[3.4843,1]]}
          RadialDensity {type=MovieSceneScriptingFloatChannel, thinned_from=5, keys [[-0.949,0],[-0.5157,0],[1.484,2],[7.1517,0.25]]}
          Radius {type=MovieSceneScriptingFloatChannel, thinned_from=6, keys [[-0.949,0],[-0.516,0],[1.484,0.4],[7.1517,0.75],[7.152,1]]}
          RingAlpha {type=MovieSceneScriptingFloatChannel, thinned_from=4, keys [[-0.949,0],[-0.516,0],[5.484,1]]}
          RingSize {type=MovieSceneScriptingFloatChannel, thinned_from=4, keys [[-0.949,0],[0.484,0],[5.484,1]]}
          RingThicc {type=MovieSceneScriptingFloatChannel, thinned_from=4, keys [[-0.949,0],[0.484,0],[5.484,1]]}
          Flare {type=MovieSceneScriptingFloatChannel, thinned_from=6, keys [[-0.949,0],[-0.568,0],[1.1507,1],[7.1517,0.75],[7.152,0]]}
          FlareAlpha {type=MovieSceneScriptingFloatChannel, thinned_from=6, keys [[-0.949,0],[-0.5677,0],[1.484,1],[3.4843,1],[7.152,0]]}
          Color_2R {type=MovieSceneScriptingFloatChannel, const [1,2]}
          Color_2G {type=MovieSceneScriptingFloatChannel, const [0.0081,2]}
          Color_2B {type=MovieSceneScriptingFloatChannel, const [0,2]}
          Color_2A {type=MovieSceneScriptingFloatChannel, const [0,2]}
          Color_1R {type=MovieSceneScriptingFloatChannel, keys [[0,1],[6,0.62]]}
          Color_1G {type=MovieSceneScriptingFloatChannel, keys [[0,0.0904],[6,0]]}
          Color_1B {type=MovieSceneScriptingFloatChannel, keys [[0,0.2911],[6,0.02]]}
          Color_1A {type=MovieSceneScriptingFloatChannel, const [0,2]}
         ]
        }]
      }
      {
       type=MovieSceneByteTrack
       display_name=Visibility
       sections [{type=MovieSceneByteSection, end_frame=8, start_frame=0, channels [None {type=MovieSceneScriptingByteChannel, keys [[0,4],[8.8543,2]]}]}]
      }
     ]
    }]}
 }
 Anim_LocationPlaced {binding_count=0, length_seconds=0.5089, object=&self:Anim_LocationPlaced, sequence {bindings []}}
 Anim_ItemPlaced {binding_count=0, length_seconds=0.5089, object=&self:Anim_ItemPlaced, sequence {bindings []}}
 Anim_Hovered {
  binding_count=1
  length_seconds=0.2
  object=&self:Anim_Hovered
  sequence {bindings [Item_Large {
     id=#2
     display_name=Image_LargeIndicator
     possessed_class=Image
     tracks [
      {
       type=MovieSceneWidgetMaterialTrack
       sections [{
         type=MovieSceneParameterSection
         end_frame=4
         start_frame=0
         channels [
          SolidFull {type=MovieSceneScriptingFloatChannel, keys [[0,0],[4,1]]}
          IndicatorShape {type=MovieSceneScriptingFloatChannel, keys [[0,0],[4,1]]}
         ]
        }]
      }
      {
       type=MovieScene2DTransformTrack
       sections [{
         type=MovieScene2DTransformSection
         end_frame=4
         start_frame=0
         channels [
          Translation.X {type=MovieSceneScriptingFloatChannel}
          Translation.Y {type=MovieSceneScriptingFloatChannel}
          Angle {type=MovieSceneScriptingFloatChannel}
          Scale.X {type=MovieSceneScriptingFloatChannel, keys [[0,1],[3,0.3318],[4,0.3318]]}
          Scale.Y {type=MovieSceneScriptingFloatChannel, keys [[0,1],[3,0.3318],[4,0.3318]]}
          Shear.X {type=MovieSceneScriptingFloatChannel}
          Shear.Y {type=MovieSceneScriptingFloatChannel}
         ]
        }]
      }
      {
       type=MovieSceneFloatTrack
       display_name=Render Opacity
       sections [{type=MovieSceneFloatSection, end_frame=4, start_frame=0, channels [None {type=MovieSceneScriptingFloatChannel, keys [[0,1],[4,0.5]]}]}]
      }
     ]
    }]}
 }
]
graphs [
 EventGraph {nodes [
   OnEnemyAttacking() {is=title K2Node_CustomEvent}
   > n1 = Set EnemyEventState(EnemyEventState:E_SaplingEnemyEventState Enum=NewEnumerator1) -> (Output_Get:E_SaplingEnemyEventState Enum=NewEnumerator0)
   UpdateColor(execute=[@n3.then, @n1.then, @n4.then])
   n2 = SaplingUpdateEvent!() -> (OutputDelegate, then, Sapling:B SMA Athena Reality Sapling Parent Object Reference) {is=title K2Node_CustomEvent}
   > n3 = SetAndBindCachedSapling!(execute, self, NewSapling:B SMA Athena Reality Sapling Parent Object Reference=@n2.Sapling) -> (then)
   OnEnemyHarvesting() {is=title K2Node_CustomEvent}
   > n4 = Set EnemyEventState(EnemyEventState=NewEnumerator2) -> (Output_Get=NewEnumerator0)
   n5 = Event OnSetMarkerData!() -> (OutputDelegate, then, FortWorldMarkerData:"Fort World Marker Data Structure (by ref)") {is=title K2Node_Event}
   > n6 = Cast To B_SMA_Athena_RealitySapling_Parent!(
    execute
    Object=Break Fort World Marker Data!(FortWorldMarkerData:"Fort World Marker Data Structure (by ref)"=@n5.FortWorldMarkerData) -> (
     MarkerID:Marker ID Structure
     Owner:Fort Player State Weak object
     MarkerType:EFortWorldMarkerType Enum
     BasePositionOffset:Vector=~
     WorldNormal:Vector=~
     ItemDefinition:Item Definition Base Object Reference
     ItemCount:int
     PinnedQuestIndex:int
     MarkedActor:Actor Weak object
     MarkedActorClass:Actor Soft Class Reference
     MarkerTag:Gameplay Tag Structure
     bUseHoveredMarkerDetail:bool=~
    ) {is=title K2Node_BreakStruct}.MarkedActor
   ) -> (then, CastFailed, AsB SMA Athena Reality Sapling Parent:B SMA Athena Reality Sapling Parent Object Reference) {is=title K2Node_DynamicCast}
   > SetAndBindCachedSapling(NewSapling=@n6)
   > Set EnemyEventState(EnemyEventState=NewEnumerator0) -> (Output_Get=NewEnumerator0)
   > UpdateColor()
  ]}
 GetEnemyEventState {nodes [
   GetEnemyEventState() {is=title K2Node_FunctionEntry}
   > Return Node!(
    execute
    EnemyEventState:E_SaplingEnemyEventState Enum=Get EnemyEventState() -> (EnemyEventState:E_SaplingEnemyEventState Enum=NewEnumerator0)
   )
  ]}
 SetAndBindCachedSapling {nodes [
   n1 = SetAndBindCachedSapling!() -> (then, NewSapling:B SMA Athena Reality Sapling Parent Object Reference) {is=title K2Node_FunctionEntry}
   > Branch(Condition="Not Equal (Object)"(
     A:Object Reference=@n1
     B:Object Reference=Get CachedRealitySapling() -> (CachedRealitySapling:B SMA Athena Reality Sapling Parent Object Reference)
    ) -> (ReturnValue:bool=~) {is=title K2Node_PromotableOperator})
   n2 = Get CachedRealitySapling()
   > n3 = Is Valid(InputObject=@n2)
   > Unbind Event from Sapling Update Event(self:B SMA Athena Reality Sapling Parent Object Reference=@n2, Delegate=Create Event()) {is=title K2Node_RemoveDelegate}
   > Unbind Event from Sapling Harvested by Enemy Event(self:B SMA Athena Reality Sapling Parent Object Reference=@n2, Delegate=Create Event()) {is=title K2Node_RemoveDelegate}
   > n4 = Unbind Event from Sapling Damaged by Enemy Event(self:B SMA Athena Reality Sapling Parent Object Reference=@n2, Delegate=Create Event()) {is=title K2Node_RemoveDelegate}
   n5 = Set CachedRealitySapling(
    execute=[@n4.then, @n3.Is Not Valid]
    CachedRealitySapling:B SMA Athena Reality Sapling Parent Object Reference=Get NewSapling() -> (NewSapling:B SMA Athena Reality Sapling Parent Object Reference)
   ) -> (Output_Get:B SMA Athena Reality Sapling Parent Object Reference)
   > Is Valid(InputObject=@n5)
   > Bind Event to Sapling Update Event(self:B SMA Athena Reality Sapling Parent Object Reference=@n5, Delegate=Create Event()) {is=title K2Node_AddDelegate}
   > Bind Event to Sapling Harvested by Enemy Event(self:B SMA Athena Reality Sapling Parent Object Reference=@n5, Delegate=Create Event()) {is=title K2Node_AddDelegate}
   > Bind Event to Sapling Damaged by Enemy Event(self:B SMA Athena Reality Sapling Parent Object Reference=@n5, Delegate=Create Event()) {is=title K2Node_AddDelegate}
  ]}
 UpdateColor {nodes [
   UpdateColor() {is=title K2Node_FunctionEntry}
   > n1 = Branch(Condition="Equal (Enum)"(A:E_SaplingEnemyEventState Enum=Get EnemyEventState() -> (EnemyEventState=NewEnumerator0), B:E_SaplingEnemyEventState Enum=NewEnumerator0))
   > Set NewColor(NewColor:Linear Color Structure=Get DefaultColor() -> (DefaultColor:Linear Color Structure)) -> (Output_Get:Linear Color Structure)
   > n2 = Sequence()
   n3 = Get CachedRealitySapling()
   > Is Valid(InputObject=@n3)
   n4 = Get FruitStates() -> (FruitStates:Array of Bytes)
   > ConvertToFruitVariantByteArray!(
    execute
    FruitSlotArray:Array of Server Spawned Fruit Slot Structures=Get ServerSpawnedFruitSlot!(self:B SMA Athena Reality Sapling Parent Object Reference=@n3) -> (ServerSpawnedFruitSlot:Array of Server Spawned Fruit Slot Structures)
    OutByteArray:Array of Bytes=@n4
   ) -> (then)
   > n5 = Get Reality Sapling Rarity Data!(
    execute
    FruitStates:Array of Bytes=@n4
    NurtureLevel:Byte="To Byte (Integer)"!(InInt:int=Get SaplingNurturedCountThisMatch!(self:B SMA Athena Reality Sapling Parent Object Reference=@n3) -> (SaplingNurturedCountThisMatch:int=0)) -> (ReturnValue:Byte=0)
   ) -> (then, IsFruitGrowing:bool, RarityName:Text, RarityColor:Linear Color Structure) {is=title K2Node_CallFunction}
   > Set NewColor(NewColor=@n5.RarityColor)
   SetColorAndOpacity(
    execute=@n2.then_1
    self:Image Object Reference=Get Image_SmallIndicator() -> (Image_SmallIndicator:Image Object Reference)
    InColorAndOpacity:Linear Color Structure=Get NewColor() -> (NewColor:Linear Color Structure)
   )
   > SetActiveWidgetIndex(self=Get Switcher_IndicatorIcon() -> (Switcher_IndicatorIcon:Widget Switcher Object Reference), Index=0)
   > n6 = GetDynamicMaterial(self=Get Image_LargeIndicator() -> (Image_LargeIndicator:Image Object Reference))
   > SetVectorParameterValue(self=@n6, ParameterName=Get MIDParam_ItemRarityColor() -> (MIDParam_ItemRarityColor:Name=None), Value=Get NewColor())
   SetActiveWidgetIndex(execute=@n1.else, self=Get Switcher_IndicatorIcon(), Index=1)
   > SetColorAndOpacity(self=Get Image_SmallIndicator(), InColorAndOpacity=(R=1,G=0,B=0,A=1))
  ]}
]
variables [
 CachedRealitySapling {type=B_SMA_Athena_RealitySapling_Parent_C, default=null}
 DefaultColor {type=LinearColor, default=<rgba 1 1 1 1>}
 EnemyEventState {type=E_SaplingEnemyEventState, default=None}
 FruitStates {type="array<integer>"}
]
widgets {
 info {inheritedWidgetCount=11, parentClass=/Script/FortniteUI.AthenaSquadIndicator, rootWidgetClass=&Overlay, widgetCount=16}
 widgets [
  Overlay_0 {widgetClassPath=&Overlay, widget=Overlay_0}
  Panel_SmallIndicator {
   widgetClassPath=&Overlay
   bInherited=true
   parent=Overlay_0
   widget=Panel_SmallIndicator
   slot {class=OverlaySlot, properties {horizontalAlignment=HAlign_Fill, verticalAlignment=VAlign_Fill}}
  }
  Image_TrajectorySmallIndicator {
   widgetClassPath=&Image
   bInherited=true
   parent=Panel_SmallIndicator
   widget=Image_TrajectorySmallIndicator
   properties {
    visibility=Collapsed
    brush {
     bIsDynamicallyLoaded=false
     drawAs=Image
     imageType=NoImage
     mirroring=NoMirror
     resourceName=None
     resourceObject=/Game/Athena/HUD/Materials/M_UI_HexSocket_Inst
     tiling=NoTile
     imageSize=<32 32>
     margin=<ltrb 0 0 0 0>
     outlineSettings {
      bUseBrushTransparency=false
      roundingType=HalfHeightRadius
      width=0
      color {colorUseRule=UseColor_Specified, specifiedColor=<rgba 0 0 0 0>}
      cornerRadii=<0 0 0 0>
     }
     tintColor {colorUseRule=UseColor_Specified, specifiedColor=<rgba 1 1 1 1>}
     uVRegion {bIsValid=false, max=<0 0>, min=<0 0>}
    }
   }
   slot {class=OverlaySlot, properties {horizontalAlignment=HAlign_Center, verticalAlignment=VAlign_Bottom}}
  }
  Image_SmallIndicator {
   widgetClassPath=&Image
   bInherited=true
   parent=Panel_SmallIndicator
   widget=Image_SmallIndicator
   properties {
    brush {
     bIsDynamicallyLoaded=false
     drawAs=Image
     imageType=NoImage
     mirroring=NoMirror
     resourceName=None
     resourceObject=/Game/Athena/HUD/Materials/M_PingIndicator
     tiling=NoTile
     imageSize=<16 16>
     margin=<ltrb 0 0 0 0>
     outlineSettings {
      bUseBrushTransparency=false
      roundingType=HalfHeightRadius
      width=0
      color {colorUseRule=UseColor_Specified, specifiedColor=<rgba 0 0 0 0>}
      cornerRadii=<0 0 0 0>
     }
     tintColor {colorUseRule=UseColor_Specified, specifiedColor=<rgba 1 1 1 1>}
     uVRegion {bIsValid=false, max=<0 0>, min=<0 0>}
    }
    renderTransform {angle=0, scale=<1 1>, shear=<0 0>, translation=<0 -3>}
   }
   slot {class=OverlaySlot, properties {horizontalAlignment=HAlign_Center, verticalAlignment=VAlign_Bottom}}
  }
  Panel_LargeIndicator {
   widgetClassPath=&Overlay
   bInherited=true
   parent=Overlay_0
   widget=Panel_LargeIndicator
   slot {class=OverlaySlot, properties {horizontalAlignment=HAlign_Fill, verticalAlignment=VAlign_Fill}}
  }
  Image_TrajectoryLargeIndicator {
   widgetClassPath=&Image
   bInherited=true
   parent=Panel_LargeIndicator
   widget=Image_TrajectoryLargeIndicator
   properties {
    visibility=Collapsed
    brush {
     bIsDynamicallyLoaded=false
     drawAs=Image
     imageType=NoImage
     mirroring=NoMirror
     resourceName=None
     resourceObject=&M_UI_HexSocket_Inst
     tiling=NoTile
     imageSize=<45 45>
     margin=<ltrb 0 0 0 0>
     outlineSettings {
      bUseBrushTransparency=false
      roundingType=HalfHeightRadius
      width=0
      color {colorUseRule=UseColor_Specified, specifiedColor=<rgba 0 0 0 0>}
      cornerRadii=<0 0 0 0>
     }
     tintColor {colorUseRule=UseColor_Specified, specifiedColor=<rgba 1 1 1 1>}
     uVRegion {bIsValid=false, max=<0 0>, min=<0 0>}
    }
   }
   slot {class=OverlaySlot}
  }
  Image_LocationFlare {
   widgetClassPath=&Image
   bInherited=true
   parent=Panel_LargeIndicator
   widget=Image_LocationFlare
   properties {
    visibility=SelfHitTestInvisible
    brush {
     bIsDynamicallyLoaded=false
     drawAs=Image
     imageType=NoImage
     mirroring=NoMirror
     resourceName=None
     resourceObject=/Game/Athena/HUD/Materials/M_UI_LocationPingReminder
     tiling=NoTile
     imageSize=<25 25>
     margin=<ltrb 0 0 0 0>
     outlineSettings {
      bUseBrushTransparency=false
      roundingType=HalfHeightRadius
      width=0
      color {colorUseRule=UseColor_Specified, specifiedColor=<rgba 0 0 0 0>}
      cornerRadii=<0 0 0 0>
     }
     tintColor {colorUseRule=UseColor_Specified, specifiedColor=<rgba 1 1 1 1>}
     uVRegion {bIsValid=false, max=<0 0>, min=<0 0>}
    }
    renderTransform {angle=0, scale=<3 3>, shear=<0 0>, translation=<0 6>}
   }
   slot {class=OverlaySlot, properties {horizontalAlignment=HAlign_Center}}
  }
  Switcher_IndicatorIcon {
   widgetClassPath=&WidgetSwitcher
   bIsVariable=true
   parent=Panel_LargeIndicator
   widget=Switcher_IndicatorIcon
   slot {class=OverlaySlot, properties {horizontalAlignment=HAlign_Fill, verticalAlignment=VAlign_Fill}}
  }
  Image_LargeIndicator {
   widgetClassPath=&Image
   bInherited=true
   parent=Switcher_IndicatorIcon
   widget=Image_LargeIndicator
   properties {brush {
     bIsDynamicallyLoaded=false
     drawAs=Image
     imageType=NoImage
     mirroring=NoMirror
     resourceName=None
     resourceObject=/Game/Athena/HUD/Materials/M_UI_Ping_Location_Test
     tiling=NoTile
     imageSize=<45 45>
     margin=<ltrb 0 0 0 0>
     outlineSettings {
      bUseBrushTransparency=false
      roundingType=HalfHeightRadius
      width=0
      color {colorUseRule=UseColor_Specified, specifiedColor=<rgba 0 0 0 0>}
      cornerRadii=<0 0 0 0>
     }
     tintColor {colorUseRule=UseColor_Specified, specifiedColor=<rgba 1 1 1 1>}
     uVRegion {bIsValid=false, max=<0 0>, min=<0 0>}
    }}
   slot {class=WidgetSwitcherSlot, properties {horizontalAlignment=HAlign_Left, verticalAlignment=VAlign_Top}}
  }
  DangerIcon {
   widgetClassPath=&Image
   parent=Switcher_IndicatorIcon
   widget=DangerIcon
   properties {
    brush {
     bIsDynamicallyLoaded=false
     drawAs=Image
     imageType=NoImage
     mirroring=NoMirror
     resourceName=None
     resourceObject=/Game/Athena/HUD/Materials/M_UI_PingBubble_Inst
     tiling=NoTile
     imageSize=<75 75>
     margin=<ltrb 0 0 0 0>
     outlineSettings {
      bUseBrushTransparency=false
      roundingType=HalfHeightRadius
      width=0
      color {colorUseRule=UseColor_Specified, specifiedColor=<rgba 0 0 0 0>}
      cornerRadii=<0 0 0 0>
     }
     tintColor {colorUseRule=UseColor_Specified, specifiedColor=<rgba 1 1 1 1>}
     uVRegion {bIsValid=false, max=<0 0>, min=<0 0>}
    }
    renderTransformPivot=<1 0.5>
   }
   slot {class=WidgetSwitcherSlot, properties {horizontalAlignment=HAlign_Left, verticalAlignment=VAlign_Top}}
  }
  Burst {
   widgetClassPath=&Image
   parent=Overlay_0
   widget=Burst
   properties {
    visibility=Hidden
    brush {
     bIsDynamicallyLoaded=false
     drawAs=Image
     imageType=NoImage
     mirroring=NoMirror
     resourceName=None
     resourceObject=/Game/Athena/HUD/Art/squadArt/M_UI_Burst_Ping
     tiling=NoTile
     imageSize=<32 32>
     margin=<ltrb 0 0 0 0>
     outlineSettings {
      bUseBrushTransparency=false
      roundingType=HalfHeightRadius
      width=0
      color {colorUseRule=UseColor_Specified, specifiedColor=<rgba 0 0 0 0>}
      cornerRadii=<0 0 0 0>
     }
     tintColor {colorUseRule=UseColor_Specified, specifiedColor=<rgba 1 1 1 1>}
     uVRegion {bIsValid=false, max=<0 0>, min=<0 0>}
    }
    renderTransform {angle=0, scale=<7 7>, shear=<0 0>, translation=<0 0>}
   }
   slot {class=OverlaySlot, properties {horizontalAlignment=HAlign_Center, verticalAlignment=VAlign_Center}}
  }
  AttachText {
   widgetClassPath=&FortVisualAttachment
   parent=Overlay_0
   widget=AttachText
   properties {renderTransform {angle=0, scale=<1 1>, shear=<0 0>, translation=<10 0>}}
   slot {class=OverlaySlot, properties {horizontalAlignment=HAlign_Center, verticalAlignment=VAlign_Center}}
  }
  Text_Distance {
   widgetClassPath=&CommonTextBlock
   bInherited=true
   parent=AttachText
   widget=Text_Distance
   properties {
    text=100m
    font {
     bForceMonospaced=false
     bMaterialIsStencil=false
     fontMaterial=null
     fontObject=&Burbank
     letterSpacing=0
     monospacedWidth=1
     size=14
     skewAmount=0
     typefaceFontName=BurbankRegular-Black
     outlineSettings {
      bApplyOutlineToDropShadows=false
      bMiteredCorners=false
      bSeparateFillAlpha=false
      outlineMaterial=&M-FontShadow
      outlineSize=3
      outlineColor=<rgba 0 0 0 0.6>
     }
    }
    renderTransform {angle=0, scale=<1 1>, shear=<-8 0>, translation=<0 0>}
    shadowColorAndOpacity=<rgba 0 0 0 0.8>
    shadowOffset=<4 3>
   }
   slot {class=SizeBoxSlot}
  }
  MarkerConfirmationWidget {widgetClassPath=/Script/FortniteUI.AthenaMarkerConfirmationWidget, bInherited=true, widget=null}
  Switcher_IndicatorSize {widgetClassPath=&WidgetSwitcher, bInherited=true, widget=null}
  TutorialHighlight {widgetClassPath=&TutorialHighlight}
 ]
}
