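# RealitySaplingToast is a prioritized HUD toast announcing that a Reality Sapling was planted and its location marked on the map.
# It listens for the SeedPlanted gameplay event on the cached pawn, starting a new wait on each view-target change.
# If the map-toggle actions are bound elsewhere, display waits until they free up.
# Showing plays the intro, claims those actions and hides on a timer; clicking the toast toggles the fullscreen map.
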WidgetBlueprint /RealitySeedGameplay/UI/RealitySaplingToast
animations [ToastIntro {
  binding_count=7
  length_seconds=0.75
  object=&self:ToastIntro
  sequence {
   bindings [
    Text_FirstCatch {
     id=#1
     possessed_class=CommonTextBlock
     tracks [
      {
       type=MovieSceneFloatTrack
       display_name=Render Opacity
       sections [{type=MovieSceneFloatSection, channels [None {type=MovieSceneScriptingFloatChannel, keys [[9,0],[12,1]]}]}]
      }
      {
       type=MovieScene2DTransformTrack
       sections [{
         type=MovieScene2DTransformSection
         channels [
          Translation.X {type=MovieSceneScriptingFloatChannel, keys [[9,-32],[18,0]]}
          Translation.Y {type=MovieSceneScriptingFloatChannel, const [0,1]}
         ]
        }]
      }
     ]
    }
    Border_Background {
     id=#2
     possessed_class=CommonBorder
     tracks [{
       type=MovieScene2DTransformTrack
       sections [{
         type=MovieScene2DTransformSection
         channels [
          Scale.X {type=MovieSceneScriptingFloatChannel, keys [[0,0.5],[15,1]]}
          Scale.Y {type=MovieSceneScriptingFloatChannel, const [1,1]}
         ]
        }]
      }]
    }
    Image_textglow_1 {
     id=#3
     possessed_class=Image
     tracks [{
       type=MovieSceneFloatTrack
       display_name=Render Opacity
       sections [{type=MovieSceneFloatSection, channels [None {type=MovieSceneScriptingFloatChannel, keys [[6,0],[12,1],[45.001,0.75]]}]}]
      }]
    }
    HorizontalBox_0 {
     id=#4
     possessed_class=HorizontalBox
     tracks [
      {
       type=MovieSceneFloatTrack
       display_name=Render Opacity
       sections [{type=MovieSceneFloatSection, channels [None {type=MovieSceneScriptingFloatChannel, keys [[6,0],[9,1]]}]}]
      }
      {
       type=MovieScene2DTransformTrack
       sections [{
         type=MovieScene2DTransformSection
         channels [
          Translation.X {type=MovieSceneScriptingFloatChannel, keys [[6,-32],[15,0]]}
          Translation.Y {type=MovieSceneScriptingFloatChannel, const [0,1]}
         ]
        }]
      }
     ]
    }
    HorizontalBoxHotkeyPrompt {
     id=#5
     possessed_class=HorizontalBox
     tracks [
      {
       type=MovieSceneFloatTrack
       display_name=Render Opacity
       sections [{type=MovieSceneFloatSection, channels [None {type=MovieSceneScriptingFloatChannel, keys [[20,0],[23,1]]}]}]
      }
      {
       type=MovieScene2DTransformTrack
       sections [{type=MovieScene2DTransformSection, channels [Translation.X {type=MovieSceneScriptingFloatChannel, keys [[20,-32],[39,0]]}]}]
      }
     ]
    }
    Overlay_ItemImage {
     id=#6
     possessed_class=Overlay
     tracks [
      {
       type=MovieSceneFloatTrack
       display_name=Render Opacity
       sections [{type=MovieSceneFloatSection, channels [None {type=MovieSceneScriptingFloatChannel, keys [[0,0],[6,1]]}]}]
      }
      {
       type=MovieScene2DTransformTrack
       sections [{
         type=MovieScene2DTransformSection
         channels [
          Translation.X {type=MovieSceneScriptingFloatChannel, const [16,1]}
          Translation.Y {type=MovieSceneScriptingFloatChannel, const [0,1]}
          Scale.X {type=MovieSceneScriptingFloatChannel, keys [[0,0.5],[6,1.05],[12,1]]}
          Scale.Y {type=MovieSceneScriptingFloatChannel, keys [[0,0.5],[6,1.05],[12,1]]}
         ]
        }]
      }
     ]
    }
    Overlay_Background {
     id=#7
     possessed_class=Overlay
     tracks [{
       type=MovieSceneFloatTrack
       display_name=Render Opacity
       sections [{type=MovieSceneFloatSection, channels [None {type=MovieSceneScriptingFloatChannel, keys [[0,0],[3,1]]}]}]
      }]
    }
   ]
   display_rate {numerator=60}
  }
 }]
graphs [EventGraph {nodes [
   "On Clicked (Button_MobileClickCatcher)"() {is=title K2Node_ComponentBoundEvent}
   > ToggleFullscreenMap(self:Fort HUDContext Object Reference=Get FortHUDContext!() -> (ReturnValue:Fort HUDContext Object Reference) {is=title K2Node_GetBlueprintContextEx})
   n1 = ToastTimerExpired() {is=title K2Node_CustomEvent}
   > StopDisplay()
   > SetHasActiveContent!(execute, self:Fort Prioritized Widget Interface, bNewActive:bool=false) -> (then)
   n2 = GetOwningPlayer()
   > Is Valid(InputObject=@n2)
   > n3 = Cast To FortPlayerController(Object=@n2)
   > SetBoundEntryToActions!(
    execute
    self:Fort Player Controller Object Reference=@n3
    InBoundActions:Array of Names=Get ButtonHandling() -> (ButtonHandling:Array of Names)
    EntryToBind:Widget Object Reference
   ) -> (then)
   StartRealitySaplingToast() {is=title K2Node_CustomEvent}
   > n4 = Branch(Condition=Get IsDisplaying() -> (IsDisplaying:bool))
   Branch(execute=@n4.else)
   > SetVisibility(InVisibility=SelfHitTestInvisible)
   > PlayAnimation(InAnimation=Get ToastIntro() -> (ToastIntro:Widget Animation Object Reference))
   > SetHasActiveContent(bNewActive=true)
   > Set IsDisplaying(IsDisplaying:bool=true) -> (Output_Get:bool)
   > Set Timer by Event(Delegate=@n1, Time=5, bLooping=false)
   n5 = GetOwningPlayer()
   > Is Valid(InputObject=@n5)
   > n6 = Cast To FortPlayerController(Object=@n5)
   > SetBoundEntryToActions(self=@n6, InBoundActions=Get ButtonHandling(), EntryToBind=Self-Reference())
   n7 = CheckAndWaitForFreedEvent!() -> (OutputDelegate, then, ActionFreed:Name) {is=title K2Node_CustomEvent}
   n8 = GetOwningPlayer()
   > Is Valid(InputObject=@n8)
   > n9 = Cast To FortPlayerController(Object=@n8)
   > n10 = GetCurrentlyBoundEntryFromActions!(execute, self:Fort Player Controller Object Reference=@n9, InBoundActionsToCheck:Array of Names=Get ButtonHandling()) -> (then, ReturnValue:Widget Object Reference)
   > n11 = Branch(Condition=IsValid(Object=@n10))
   Unbind Event from On Bound Action Freed(execute=@n11.else, self:Fort Player Controller Object Reference=@n9, Delegate=@n7) {is=title K2Node_RemoveDelegate}
   > StartRealitySaplingToast() {is=title K2Node_CallFunction}
   StopDisplay() {is=title K2Node_CustomEvent}
   > SetVisibility(InVisibility=Collapsed)
   > Set IsDisplaying(IsDisplaying=false)
   SetBinds() {is=title K2Node_CustomEvent}
   > StopDisplay() {is=title K2Node_CallFunction}
   > n12 = WaitGameplayEventToActor!(
    execute:Exec
    TargetActor:Actor Object Reference=Get Cached Pawn to Use() -> (Cached Pawn to Use:Fort Player Pawn Athena Object Reference)
    EventTag:Gameplay Tag Structure=(TagName="Athena.Event.Gameplay.SeedPlanted")
    OnlyTriggerOnce:bool=false
    OnlyMatchExact:bool=true
   ) -> (then:Exec, AsyncTaskProxy:Ability Async Wait Gameplay Event Object Reference, EventReceived:Exec, Payload:Gameplay Event Data Structure) {is=title K2Node_AsyncAction}
   n13 = GetOwningPlayer()
   Is Valid(exec=@n12.EventReceived, InputObject=@n13)
   > n14 = Cast To FortPlayerController(Object=@n13)
   > n15 = GetCurrentlyBoundEntryFromActions(self=@n14, InBoundActionsToCheck=Get ButtonHandling())
   > n16 = Branch(Condition=IsValid(Object=@n15))
   > Bind Event to On Bound Action Freed(self:Fort Player Controller Object Reference=@n14, Delegate=@n7) {is=title K2Node_AddDelegate}
   StartRealitySaplingToast(execute=@n16.else)
   n17 = OnViewTargetChanged!() -> (
    OutputDelegate
    then
    PlayerController:Fort Player Controller Object Reference
    OldViewTarget:Actor Object Reference
    NewViewTarget:Actor Object Reference
   ) {is=title K2Node_CustomEvent}
   > n18 = Cast To FortPlayerPawnAthena(Object=@n17.NewViewTarget)
   > Set Cached Pawn to Use(Cached Pawn to Use:Fort Player Pawn Athena Object Reference=@n18) -> (Output_Get:Fort Player Pawn Athena Object Reference)
   > SetBinds() {is=title K2Node_CallFunction}
   Event Construct()
   > SetVisibility(InVisibility=Collapsed)
   > n19 = Sequence()
   > n20 = Cast To FortPlayerController(Object=GetOwningPlayer())
   > Bind Event to On View Target Changed(self:Fort Player Controller Object Reference=@n20, Delegate=@n17) {is=title K2Node_AddDelegate}
   n21 = GetOwningPlayerPawn!(self:User Widget Object Reference) -> (ReturnValue:Pawn Object Reference)
   > n22 = Is Valid(InputObject=@n21)
   PrintString!(execute=@n22.Is Not Valid, InString="RealitySaplingToast: No Owning Player Pawn!!", bPrintToScreen, bPrintToLog, TextColor, Duration, Key) -> (then)
   n23 = Cast To FortPlayerPawnAthena(execute=@n22.Is Valid, Object=@n21)
   > Set Cached Pawn to Use(Cached Pawn to Use=@n23)
   > SetBinds()
   PrintString(execute=@n20.CastFailed, InString="RealitySaplingToast: FortPlayerController cast failed!")
   Branch(execute=@n19.then_1, Condition=IsUsingTouch(PlayerController=GetOwningPlayer()))
   > SetPadding!(
    execute
    self:Overlay Slot Object Reference=SlotAsOverlaySlot(Widget=Get Overlay_TopContainer() -> (Overlay_TopContainer:Overlay Object Reference))
    InPadding:Margin Structure=Make Margin(Left=0, Top=25, Right=0, Bottom=0)
   ) -> (then)
  ]}]
variables [
 ButtonHandling {type="array<string>", default [ToggleFullScreenMap, GamepadToggleFullScreenMap]}
 Cached Pawn to Use {type=FortPlayerPawnAthena}
 IsDisplaying {type=boolean}
]
widgets {
 info {inheritedWidgetCount=1, parentClass=/Script/FortniteUI.GenericFortPrioritizedWidgetBase, rootWidgetClass=&Overlay, widgetCount=30}
 widgets [
  Overlay_18 {
   widgetClassPath=&Overlay
   widget=Overlay_18
   properties {renderTransform {angle=0, scale=<1 1>, shear=<0 0>, translation=<0 -100>}}
  }
  Overlay_TopContainer {
   widgetClassPath=&Overlay
   bIsVariable=true
   parent=Overlay_18
   widget=Overlay_TopContainer
   slot {class=OverlaySlot, properties {horizontalAlignment=HAlign_Center, verticalAlignment=VAlign_Center}}
  }
  Overlay_Background {
   widgetClassPath=&Overlay
   parent=Overlay_TopContainer
   widget=Overlay_Background
   properties {visibility=Visible, renderTransformPivot=<0 0.5>}
   slot {class=OverlaySlot, properties {horizontalAlignment=HAlign_Fill, verticalAlignment=VAlign_Fill, padding=<ltrb 0 20 0 0>}}
  }
  SizeBox_0 {
   widgetClassPath=&SizeBox
   parent=Overlay_Background
   widget=SizeBox_0
   properties {
    bOverride_HeightOverride=true
    heightOverride=73
    renderTransform {angle=0, scale=<1 1>, shear=<0 0>, translation=<0 -9>}
    renderTransformPivot=<0 0.5>
   }
   slot {class=OverlaySlot, properties {horizontalAlignment=HAlign_Fill, verticalAlignment=VAlign_Center}}
  }
  Border_Background {
   widgetClassPath=&CommonBorder
   parent=SizeBox_0
   widget=Border_Background
   properties {
    background {
     bIsDynamicallyLoaded=false
     drawAs=Box
     imageType=NoImage
     mirroring=NoMirror
     resourceName=None
     resourceObject=/Game/Athena/HUD/Collections/Materials/M_UI_BackgroundWShadow_Inst
     tiling=NoTile
     imageSize=<128 128>
     margin=<ltrb 0.2 0 0.25 0>
     outlineSettings {
      bUseBrushTransparency=false
      roundingType=HalfHeightRadius
      width=0
      color {colorUseRule=UseColor_Specified, specifiedColor=<rgba 0 0 0 0>}
      cornerRadii=<0 0 0 0>
     }
     tintColor {colorUseRule=UseColor_Specified, specifiedColor=<rgba 1 1 1 1>}
     uVRegion {bIsValid=false, max=<0 0>, min=<0 0>}
    }
    padding=<ltrb 0 0 0 0>
    renderTransformPivot=<0 0.5>
   }
   slot {class=SizeBoxSlot, properties {padding=<ltrb 0 0 -40 -10>}}
  }
  Button_MobileClickCatcher {
   widgetClassPath=&Button
   parent=Border_Background
   widget=Button_MobileClickCatcher
   properties {isFocusable=false, backgroundColor=<rgba 1 1 1 0>, colorAndOpacity=<rgba 1 1 1 0>}
   slot {class=BorderSlot, properties {padding=<ltrb 0 0 0 0>}}
  }
  HorizontalBoxHotkeyPrompt {
   widgetClassPath=&HorizontalBox
   parent=Overlay_Background
   widget=HorizontalBoxHotkeyPrompt
   properties {visibility=Visible, renderTransform {angle=0, scale=<1 1>, shear=<0 0>, translation=<0 16>}}
   slot {class=OverlaySlot, properties {horizontalAlignment=HAlign_Right, verticalAlignment=VAlign_Bottom}}
  }
  SizeBox_Hotkey {
   widgetClassPath=&SizeBox
   parent=HorizontalBoxHotkeyPrompt
   widget=SizeBox_Hotkey
   properties {bOverride_HeightOverride=true, bOverride_WidthOverride=true, heightOverride=32, widthOverride=32}
   slot {class=HorizontalBoxSlot, properties {horizontalAlignment=HAlign_Right, verticalAlignment=VAlign_Top}}
  }
  ScaleBox_0 {widgetClassPath=&ScaleBox, parent=SizeBox_Hotkey, widget=ScaleBox_0, slot {class=SizeBoxSlot}}
  Overlay_1 {widgetClassPath=&Overlay, parent=ScaleBox_0, widget=Overlay_1, slot {class=ScaleBoxSlot}}
  Gamepad {
   widgetClassPath=&CommonHardwareVisibilityBorder
   parent=Overlay_1
   widget=Gamepad
   properties {visibilityQuery {
     autoDescription=" ANY( Input.Gamepad )"
     tokenStreamVersion=0
     userDescription=""
     queryTokenStream [0,1,1,1,0]
     tagDictionary [{tagName=Input.Gamepad}]
    }}
   slot {class=OverlaySlot}
  }
  KeybindWidget {
   widgetClassPath=/FortUILibrary/Input/KeybindWidget
   parent=Gamepad
   widget=KeybindWidget
   properties {boundAction=GamepadToggleFullScreenMap, forcedHoldKeybindStatus=ForcedHold}
   slot {class=BorderSlot, properties {padding=<ltrb 0 0 0 0>}}
  }
  MouseKeyboard {
   widgetClassPath=&CommonHardwareVisibilityBorder
   parent=Overlay_1
   widget=MouseKeyboard
   properties {visibilityQuery {
     autoDescription=" ANY( Input.MouseAndKeyboard )"
     tokenStreamVersion=0
     userDescription=""
     queryTokenStream [0,1,1,1,0]
     tagDictionary [{tagName=Input.MouseAndKeyboard}]
    }}
   slot {class=OverlaySlot, properties {horizontalAlignment=HAlign_Fill, verticalAlignment=VAlign_Center}}
  }
  KeybindWidget_1 {
   widgetClassPath=&KeybindWidget
   parent=MouseKeyboard
   widget=KeybindWidget_1
   properties {boundAction=ToggleFullScreenMap, forcedHoldKeybindStatus=ForcedHold, boundActionFallbacks [GamepadToggleFullScreenMap]}
   slot {class=BorderSlot, properties {padding=<ltrb 0 0 0 0>}}
  }
  Visibility_MobileCollectionText {
   widgetClassPath=&CommonHardwareVisibilityBorder
   parent=HorizontalBoxHotkeyPrompt
   widget=Visibility_MobileCollectionText
   properties {
    padding=<ltrb 4 2 4 2>
    visibilityQuery {
     autoDescription=" ANY( Input.Touch )"
     tokenStreamVersion=0
     userDescription=""
     queryTokenStream [0,1,1,1,0]
     tagDictionary [{tagName=Input.Touch}]
    }
   }
   slot {class=HorizontalBoxSlot, properties {horizontalAlignment=HAlign_Right, verticalAlignment=VAlign_Center}}
  }
  Switcher_MobileClueText {
   widgetClassPath=&CommonWidgetSwitcherLegacy
   bIsVariable=true
   parent=Visibility_MobileCollectionText
   widget=Switcher_MobileClueText
   properties {transitionDuration=0}
   slot {class=BorderSlot}
  }
  Text_MobileCollectionPrompt {
   widgetClassPath=&CommonTextBlock
   bIsVariable=true
   parent=Switcher_MobileClueText
   widget=Text_MobileCollectionPrompt
   properties {
    text=TAP TO OPEN MAP
    textTransformPolicy=ToUpper
    font {
     bForceMonospaced=false
     bMaterialIsStencil=false
     fontMaterial=null
     fontObject=&Burbank
     letterSpacing=0
     monospacedWidth=1
     size=12
     skewAmount=0
     typefaceFontName=Burbank-PlayerPower
     outlineSettings {
      bApplyOutlineToDropShadows=false
      bMiteredCorners=false
      bSeparateFillAlpha=false
      outlineMaterial=null
      outlineSize=0
      outlineColor=<rgba 0 0 0 1>
     }
    }
    renderTransform {angle=0, scale=<1 1>, shear=<-12 0>, translation=<0 0>}
    shadowOffset=<0 0>
   }
   slot {class=WidgetSwitcherSlot}
  }
  Visibility_CollectionText {
   widgetClassPath=&CommonHardwareVisibilityBorder
   parent=HorizontalBoxHotkeyPrompt
   widget=Visibility_CollectionText
   properties {
    padding=<ltrb 4 2 4 2>
    visibilityQuery {
     autoDescription=" ANY( Input.MouseAndKeyboard, Input.Gamepad )"
     tokenStreamVersion=0
     userDescription=""
     queryTokenStream [0,1,1,2,0,1]
     tagDictionary [{tagName=Input.MouseAndKeyboard}, {tagName=Input.Gamepad}]
    }
   }
   slot {class=HorizontalBoxSlot, properties {horizontalAlignment=HAlign_Right, verticalAlignment=VAlign_Center, padding=<ltrb 8 0 0 0>}}
  }
  Text_MapPrompt {
   widgetClassPath=&CommonTextBlock
   bIsVariable=true
   parent=Visibility_CollectionText
   widget=Text_MapPrompt
   properties {
    mobileFontSizeMultiplier=1.2
    text=See Sapling Status
    textTransformPolicy=ToUpper
    font {
     bForceMonospaced=false
     bMaterialIsStencil=false
     fontMaterial=null
     fontObject=&BurbankSmall
     letterSpacing=0
     monospacedWidth=1
     size=13
     skewAmount=0
     typefaceFontName=Black
     outlineSettings {
      bApplyOutlineToDropShadows=false
      bMiteredCorners=false
      bSeparateFillAlpha=false
      outlineMaterial=null
      outlineSize=2
      outlineColor=<rgba 0 0 0 0.125>
     }
    }
    shadowOffset=<0 0>
   }
   slot {class=BorderSlot}
  }
  HorizontalBox_ImageAndDetails {
   widgetClassPath=&HorizontalBox
   parent=Overlay_TopContainer
   widget=HorizontalBox_ImageAndDetails
   slot {class=OverlaySlot, properties {horizontalAlignment=HAlign_Fill, verticalAlignment=VAlign_Fill}}
  }
  Overlay_ItemImage {
   widgetClassPath=&Overlay
   parent=HorizontalBox_ImageAndDetails
   widget=Overlay_ItemImage
   properties {renderTransform {angle=0, scale=<1 1>, shear=<0 0>, translation=<16 0>}}
   slot {class=HorizontalBoxSlot, properties {horizontalAlignment=HAlign_Center, verticalAlignment=VAlign_Center}}
  }
  Image_textglow_1 {
   widgetClassPath=&Image
   parent=Overlay_ItemImage
   widget=Image_textglow_1
   properties {
    renderOpacity=0.8
    visibility=SelfHitTestInvisible
    brush {
     bIsDynamicallyLoaded=false
     drawAs=Image
     imageType=NoImage
     mirroring=NoMirror
     resourceName=None
     resourceObject=/Game/Athena/HUD/LTM/Wax/Materials/M_UI_TextGlow
     tiling=NoTile
     imageSize=<128 128>
     margin=<ltrb 0 0 0 0>
     outlineSettings {
      bUseBrushTransparency=false
      roundingType=HalfHeightRadius
      width=0
      color {colorUseRule=UseColor_Specified, specifiedColor=<rgba 0 0 0 0>}
      cornerRadii=<0 0 0 0>
     }
     tintColor {colorUseRule=UseColor_Specified, specifiedColor=<rgba 0 0.1667 1 1>}
     uVRegion {bIsValid=false, max=<0 0>, min=<0 0>}
    }
    renderTransform {angle=2, scale=<1.25 1.125>, shear=<0 0>, translation=<-2 0>}
   }
   slot {class=OverlaySlot, properties {horizontalAlignment=HAlign_Fill, verticalAlignment=VAlign_Fill}}
  }
  LazyImage_ItemImage {
   widgetClassPath=&FortLazyImage
   parent=Overlay_ItemImage
   widget=LazyImage_ItemImage
   properties {
    materialTextureParamName=Texture
    brush {
     bIsDynamicallyLoaded=false
     drawAs=Image
     imageType=NoImage
     mirroring=NoMirror
     resourceName=None
     resourceObject=/RealitySeedGameplay/Icons/T-T-Icon-BR-PersistentRealitySapling
     tiling=NoTile
     imageSize=<128 128>
     margin=<ltrb 0 0 0 0>
     outlineSettings {
      bUseBrushTransparency=false
      roundingType=HalfHeightRadius
      width=0
      color {colorUseRule=UseColor_Specified, specifiedColor=<rgba 0 0 0 0>}
      cornerRadii=<0 0 0 0>
     }
     tintColor {colorUseRule=UseColor_Specified, specifiedColor=<rgba 1 1 1 1>}
     uVRegion {bIsValid=false, max=<0 0>, min=<0 0>}
    }
    renderTransform {angle=0, scale=<1 1>, shear=<0 0>, translation=<0 -26>}
    shadowColorAndOpacity=<rgba 0.0052 0.0052 0.0056 0.5>
   }
   slot {class=OverlaySlot, properties {horizontalAlignment=HAlign_Center, verticalAlignment=VAlign_Center}}
  }
  Border_ItemDetails {
   widgetClassPath=&CommonBorder
   parent=HorizontalBox_ImageAndDetails
   widget=Border_ItemDetails
   properties {clipping=ClipToBounds, padding=<ltrb 8 0 8 0>}
   slot {class=HorizontalBoxSlot, properties {verticalAlignment=VAlign_Center}}
  }
  VerticalBox_1 {
   widgetClassPath=&VerticalBox
   parent=Border_ItemDetails
   widget=VerticalBox_1
   slot {class=BorderSlot, properties {padding=<ltrb 8 0 8 0>}}
  }
  HorizontalBox_0 {widgetClassPath=&HorizontalBox, parent=VerticalBox_1, widget=HorizontalBox_0, slot {class=VerticalBoxSlot}}
  Text_FirstCatch_1 {
   widgetClassPath=&CommonTextBlock
   bIsVariable=true
   parent=HorizontalBox_0
   widget=Text_FirstCatch_1
   properties {
    mobileFontSizeMultiplier=1.2
    text=Location Marked
    textTransformPolicy=ToUpper
    visibility=HitTestInvisible
    colorAndOpacity {colorUseRule=UseColor_Specified, specifiedColor=<rgba 0 0.1022 1 1>}
    font {
     bForceMonospaced=false
     bMaterialIsStencil=false
     fontMaterial=/Game/Athena/HUD/Materials/M_UI_OutlineShine_White_Manual
     fontObject=&BurbankSmall
     letterSpacing=0
     monospacedWidth=1
     size=13
     skewAmount=0
     typefaceFontName=Black
     outlineSettings {
      bApplyOutlineToDropShadows=false
      bMiteredCorners=false
      bSeparateFillAlpha=false
      outlineMaterial=null
      outlineSize=0
      outlineColor=<rgba 0 0 0 1>
     }
    }
    renderTransform {angle=0, scale=<1 1>, shear=<-12 0>, translation=<0 0>}
    renderTransformPivot=<0 0.75>
    shadowOffset=<0 0>
   }
   slot {class=HorizontalBoxSlot}
  }
  Image_174 {
   widgetClassPath=&Image
   bIsVariable=true
   parent=HorizontalBox_0
   widget=Image_174
   properties {
    brush {
     bIsDynamicallyLoaded=false
     drawAs=Image
     imageType=NoImage
     mirroring=NoMirror
     resourceName=None
     resourceObject=/Game/Athena/HUD/Art/icon_map
     tiling=NoTile
     imageSize=<20 20>
     margin=<ltrb 0 0 0 0>
     outlineSettings {
      bUseBrushTransparency=false
      roundingType=HalfHeightRadius
      width=0
      color {colorUseRule=UseColor_Specified, specifiedColor=<rgba 0 0 0 0>}
      cornerRadii=<0 0 0 0>
     }
     tintColor {colorUseRule=UseColor_Specified, specifiedColor=<rgba 0 0.25 1 1>}
     uVRegion {bIsValid=false, max=<0 0>, min=<0 0>}
    }
    renderTransform {angle=0, scale=<1 1>, shear=<-12 0>, translation=<0 0>}
   }
   slot {class=HorizontalBoxSlot, properties {horizontalAlignment=HAlign_Center, verticalAlignment=VAlign_Center, padding=<ltrb 4 0 0 0>}}
  }
  Text_FirstCatch {
   widgetClassPath=&CommonTextBlock
   bIsVariable=true
   parent=VerticalBox_1
   widget=Text_FirstCatch
   properties {
    mobileFontSizeMultiplier=1.1
    text="Sapling Planted!"
    textTransformPolicy=ToUpper
    visibility=HitTestInvisible
    colorAndOpacity {colorUseRule=UseColor_Specified, specifiedColor=<rgba 0 0.5612 1 1>}
    font {
     bForceMonospaced=false
     bMaterialIsStencil=false
     fontMaterial=&M_UI_OutlineShine_White_Manual
     fontObject=&Burbank
     letterSpacing=0
     monospacedWidth=1
     size=18
     skewAmount=0
     typefaceFontName=BurbankRegular-Black
     outlineSettings {
      bApplyOutlineToDropShadows=false
      bMiteredCorners=false
      bSeparateFillAlpha=false
      outlineMaterial=null
      outlineSize=0
      outlineColor=<rgba 0 0 0 1>
     }
    }
    renderTransform {angle=0, scale=<1 1>, shear=<-12 0>, translation=<-4 0>}
    renderTransformPivot=<0 0.75>
    shadowOffset=<0 0>
   }
   slot {class=VerticalBoxSlot}
  }
  TutorialHighlight {widgetClassPath=&TutorialHighlight}
 ]
}
