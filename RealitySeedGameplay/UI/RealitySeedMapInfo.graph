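# RealitySeedMapInfo is a HUD map-info widget showing the owning player's Reality Sapling status: weeding or ghost icon, enemy attack/looting alert, fruit rarity and growth.
# It binds to the sapling controller component's change event, hiding itself when no sapling is found.
# Each change rebuilds fruit icons from a scalable float count, then UpdateSaplingData refreshes every section.
# The weeding countdown derives from the UTC clock's time-of-day, ignoring NextWeedingTime.
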WidgetBlueprint /RealitySeedGameplay/UI/RealitySeedMapInfo
graphs [
 EventGraph {nodes [
   n1 = OnOwnedSaplingsChanged() {is=title K2Node_CustomEvent}
   n2 = Get Sapling Data!(execute:Exec=[@n7.then, @n1.then]) -> (Found:Exec, Not Found:Exec, SaplingData:Owned Reality Sapling Data Structure) {is=title K2Node_MacroInstance}
   > SetVisibility(InVisibility=HitTestInvisible)
   > n3 = Sequence()
   > ClearChildren(self=Get HorizontalBox_Fruit() -> (HorizontalBox_Fruit:Horizontal Box Object Reference))
   > For Loop(
    FirstIndex=1
    LastIndex=GetValueAsInteger!(Input:"Scalable Float Structure (by ref)"=Get NumFruitsToGrow() -> (NumFruitsToGrow:Scalable Float Structure), Level:float=0) -> (ReturnValue:int=0)
   )
   > n4 = Create Widget!(
    execute:Exec
    Class:User Widget Class Reference=&/RealityFruit.RealityFruit_C
    OwningPlayer:Player Controller Object Reference
   ) -> (then:Exec, ReturnValue:Reality Fruit Object Reference)
   > AddChildToHorizontalBox!(execute, self:Horizontal Box Object Reference=Get HorizontalBox_Fruit(), Content:Widget Object Reference=@n4) -> (then, ReturnValue:Horizontal Box Slot Object Reference)
   UpdateSaplingData(execute=@n3.then_1)
   SetVisibility(execute=@n2.Not Found, InVisibility=Collapsed)
   Event Construct()
   > SetVisibility(InVisibility=Collapsed)
   n5 = GetOwningPlayer()
   > Is Valid(InputObject=@n5)
   n6 = GetComponentByClass!(
    self:Actor Object Reference=@n5
    ComponentClass:Actor Component Class Reference=/Script/RealitySeedGameplayRuntime.FortControllerComponent_RealitySapling
   ) -> (ReturnValue:Fort Controller Component Reality Sapling Object Reference)
   > Is Valid(InputObject=@n6)
   > Sequence()
   > n7 = Bind Event to On Owned Saplings Changed(self:Fort Controller Component Reality Sapling Object Reference=@n6, Delegate=@n1) {is=title K2Node_AddDelegate}
  ]}
 UpdateSaplingData {nodes [
   UpdateSaplingData() {is=title K2Node_FunctionEntry}
   > n1 = Get Sapling Data()
   > Set CurrentSaplingData(CurrentSaplingData:Owned Reality Sapling Data Structure=@n1) -> (Output_Get:Owned Reality Sapling Data Structure)
   > n2 = Sequence!(execute) -> (then_0, then_1, then_2:Exec, then_3:Exec, then_4:Exec)
   n3 = Break Owned Reality Sapling Data!(OwnedRealitySaplingData:"Owned Reality Sapling Data Structure (by ref)"=Get CurrentSaplingData() -> (CurrentSaplingData:Owned Reality Sapling Data Structure)) -> (
    Sapling:Building Reality Sapling Object Reference
    Location:Vector=~
    bHasSaplingOnAuthority:bool=~
    FruitStates:Array of Bytes
    NurturingLevel:Byte
    bNeedsWeeding:bool=~
    bInGhostState:bool=~
    bEnemyAttacking:bool=~
    bEnemyLooting:bool=~
    NextWeedingTime:Date Time Structure
   ) {is=title K2Node_BreakStruct}
   > n4 = Get Reality Sapling Rarity Data!(execute, FruitStates:Array of Bytes=@n3.FruitStates, NurtureLevel:Byte=@n3.NurturingLevel) -> (then, IsFruitGrowing:bool, RarityName:Text, RarityColor:Linear Color Structure) {is=title K2Node_CallFunction}
   n5 = Get TextBlock_FruitRarity() -> (TextBlock_FruitRarity:Common Text Object Reference)
   > "SetText (Text)"(
    self=@n5
    InText=Select(Option 0:Text=Get NoFruitText() -> (NoFruitText:Text), Option 1:Text=@n4.RarityName, Index=@n4) -> (ReturnValue:Text)
   )
   > SetColorAndOpacity(self=@n5, InColorAndOpacity=Make Slate Color(SpecifiedColor=@n4.RarityColor))
   n6 = Break Owned Reality Sapling Data(OwnedRealitySaplingData=Get CurrentSaplingData()) -> (Location=~, bHasSaplingOnAuthority=~, bNeedsWeeding=~, bInGhostState=~, bEnemyAttacking=~, bEnemyLooting=~)
   n7 = SelectInt(A=1, B=0, bPickA=@n6.bNeedsWeeding)
   n8 = Set StatusIconMaterial(
    execute=@n2.then_1
    StatusIconMaterial:Material Instance Object Reference="Get (a copy)"(
     Array:Array of Material Instance Object References=Get StatusIcons_Materials() -> (StatusIcons_Materials:Array of Material Instance Object References)
     Dimension 1=SelectInt(A=2, B=@n7, bPickA=@n6.bInGhostState)
    ) -> (Output:Material Instance Object Reference)
   ) -> (Output_Get:Material Instance Object Reference)
   > n9 = Is Valid(InputObject=@n8)
   n10 = Get Image_StatusIcon() -> (Image_StatusIcon:Image Object Reference)
   SetVisibility(execute=@n9.Is Not Valid, self=@n10, InVisibility=Hidden)
   SetVisibility(execute=@n9.Is Valid, self=@n10, InVisibility=SelfHitTestInvisible)
   > SetBrushFromMaterial(self=Get Image_StatusIcon(), Material=Get StatusIconMaterial() -> (StatusIconMaterial:Material Instance Object Reference))
   n11 = Break Owned Reality Sapling Data(OwnedRealitySaplingData=Get CurrentSaplingData()) -> (Location=~, bHasSaplingOnAuthority=~, bNeedsWeeding=~, bInGhostState=~, bEnemyAttacking=~, bEnemyLooting=~)
   n12 = Branch(execute=@n2.then_2, Condition=@n11.bEnemyAttacking)
   n13 = Get HorizontalBox_EnemyEvents() -> (HorizontalBox_EnemyEvents:Horizontal Box Object Reference)
   > SetVisibility(self=@n13, InVisibility=SelfHitTestInvisible)
   n14 = "Get (a ref)"!(
    Array:Array of Sapling Text To Material Struct Structures=Get EnemyEventData() -> (EnemyEventData:Array of Sapling Text To Material Struct Structures)
    Dimension 1=0
   ) -> (Output_Text_2_#1:Text, Output_Material_5_#2:Material Instance Object Reference)
   > "SetText (Text)"(self=Get TextBlock_EnemyEvent() -> (TextBlock_EnemyEvent:Common Text Object Reference), InText=@n14)
   > SetBrushFromMaterial(
    self=Get Image_EnemyEventIcon() -> (Image_EnemyEventIcon:Image Object Reference)
    Material=@n14.Output_Material_5_#2
   )
   n15 = Branch(execute=@n12.else, Condition=@n11.bEnemyLooting)
   > SetVisibility(self=@n13, InVisibility=SelfHitTestInvisible)
   n16 = "Get (a ref)"(Array=Get EnemyEventData(), Dimension 1=1)
   > "SetText (Text)"(self=Get TextBlock_EnemyEvent(), InText=@n16)
   > SetBrushFromMaterial(self=Get Image_EnemyEventIcon(), Material=@n16.Output_Material_5_#2)
   SetVisibility(execute=@n15.else, self=@n13, InVisibility=Collapsed)
   n17 = For Each Loop(Exec=@n2.then_3, Array:Array of Object References=GetAllChildren(self=Get HorizontalBox_Fruit())) -> (Array Element:Object Reference)
   n18 = For Each Loop(
    Exec=@n17.Completed
    Array:Array of Bytes=Break Owned Reality Sapling Data(OwnedRealitySaplingData=Get CurrentSaplingData()) -> (Location=~, bHasSaplingOnAuthority=~, bNeedsWeeding=~, bInGhostState=~, bEnemyAttacking=~, bEnemyLooting=~).FruitStates
   ) -> (Array Element:Byte)
   n19 = Get HorizontalBox_Fruit()
   > Branch(Condition="integer > integer"(A=GetChildrenCount!(self:Panel Widget Object Reference=@n19) -> (ReturnValue:int=0), B=@n18.Array Index))
   > n20 = Cast To RealityFruit!(execute, Object=GetChildAt!(self:Panel Widget Object Reference=@n19, Index:int=@n18.Array Index) -> (ReturnValue:Widget Object Reference)) -> (then, CastFailed, AsReality Fruit:Reality Fruit Object Reference) {is=title K2Node_DynamicCast}
   > n21 = Check Sapling Fruit Status!(
    execute:Exec
    FruitVariant:SaplingFruitVariants Enum=Byte to Enum SaplingFruitVariants!(Byte:Byte=@n18) -> (ReturnValue:SaplingFruitVariants Enum) {is=title K2Node_CastByteToEnum}
   ) -> (NoFruit:Exec, IsGrowing:Exec, IsReady:Exec, RarityColor:Linear Color Structure) {is=title K2Node_MacroInstance}
   SetVisibility(execute=@n21.IsGrowing, self=@n20, InVisibility=HitTestInvisible)
   > n22 = SetReady!(execute, self:Reality Fruit Object Reference=@n20, Ready:bool=false) -> (then)
   SetRarityColor!(execute=[@n24.then, @n23.then, @n22.then], self:Reality Fruit Object Reference=@n20, Color:Linear Color Structure=@n21) -> (then)
   SetVisibility(execute=@n21.IsReady, self=@n20, InVisibility=HitTestInvisible)
   > n23 = SetReady(self=@n20, Ready=true)
   SetVisibility(execute=@n21.NoFruit, self=@n20, InVisibility=HitTestInvisible)
   > n24 = SetPicked!(execute, self:Reality Fruit Object Reference=@n20, Picked:bool=true) -> (then)
   n25 = Cast To RealityFruit(execute=@n17.LoopBody, Object=@n17)
   > SetPicked(self=@n25, Picked=true)
   > SetVisibility(self=@n25, InVisibility=Collapsed)
   n26 = Branch(
    execute=@n2.then_4
    Condition=Break Owned Reality Sapling Data(OwnedRealitySaplingData=Get CurrentSaplingData()) -> (Location=~, bHasSaplingOnAuthority=~, bNeedsWeeding=~, bInGhostState=~, bEnemyAttacking=~, bEnemyLooting=~).bNeedsWeeding
   )
   > SetActiveWidgetIndex(self=Get WidgetSwitcher_WeedingStatus() -> (WidgetSwitcher_WeedingStatus:Widget Switcher Object Reference), Index=0)
   SetActiveWidgetIndex(execute=@n26.else, self=Get WidgetSwitcher_WeedingStatus(), Index=1)
   n27 = BreakDateTime!(InDateTime:Date Time Structure=UTC Now!() -> (ReturnValue:Date Time Structure)) -> (Year:int=0, Month:int=0, Day:int=0, Hour:int=0, Minute:int=0, Second:int=0, Millisecond:int=0)
   > SetTimespanValue!(
    execute
    self:Common Date Time Text Block Object Reference=Get CommonDateTimeTextBlock_NextWeedTime() -> (CommonDateTimeTextBlock_NextWeedTime:Common Date Time Text Block Object Reference)
    InTimespan:Timespan Structure=MakeTimespan!(
     Days:int=0
     Hours:int=int - int(A=24, B=@n27.Hour)
     Minutes:int=int - int(A=60, B=@n27.Minute)
     Seconds:int=int - int(A=60, B=@n27.Second)
     Milliseconds:int=0
    ) -> (ReturnValue:Timespan Structure)
   ) -> (then)
  ]}
]
variables [
 CachedOwnedSapling {type=BuildingRealitySapling, default=null}
 EnemyEventData {
  type="array<SaplingTextToMaterialStruct>"
  default [
   {material=/RealitySeedGameplay/UI/Icons/M_UI_EnemyAttacking, text=Under Attack}
   {material=/RealitySeedGameplay/UI/Icons/M_UI_EnemyLooting, text=Enemy Looting}
  ]
 }
 NoFruitText {type=string, default=None}
 NumFruitsToGrow {
  type=ScalableFloat
  default {
   registryType=None
   value=1
   curve {curveTable=/RealitySeedGameplay/DataTables/RealitySeedGameplayGameData, rowName=Default.RealitySapling.NumFruitsToGrow}
  }
 }
 StatusIcons_Materials {
  type="array<MaterialInstance>"
  default [null, /RealitySeedGameplay/UI/Icons/M_UI_NeedsWeeds_1_NoShadow, /RealitySeedGameplay/UI/Icons/M_UI_SaplingDead_NoShadow]
 }
]
widgets {
 info {inheritedWidgetCount=1, parentClass=&FortHUDElementWidget, rootWidgetClass=&VerticalBox, widgetCount=21}
 widgets [
  VerticalBox_0 {widgetClassPath=&VerticalBox, widget=VerticalBox_0}
  HorizontalBox_3 {
   widgetClassPath=&HorizontalBox
   parent=VerticalBox_0
   widget=HorizontalBox_3
   properties {flowDirectionPreference=Culture}
   slot {class=VerticalBoxSlot, properties {padding=<ltrb 0 20 0 10>}}
  }
  Text_RealitySapling {
   widgetClassPath=&CommonTextBlock
   parent=HorizontalBox_3
   widget=Text_RealitySapling
   properties {
    text=Sapling Status
    textTransformPolicy=ToUpper
    colorAndOpacity {colorUseRule=UseColor_Specified, specifiedColor=<rgba 0.2122 0.6514 1 1>}
    font {
     bForceMonospaced=false
     bMaterialIsStencil=false
     fontMaterial=null
     fontObject=&BurbankSmall
     letterSpacing=0
     monospacedWidth=1
     size=18
     skewAmount=0
     typefaceFontName=Bold
     outlineSettings {
      bApplyOutlineToDropShadows=false
      bMiteredCorners=false
      bSeparateFillAlpha=false
      outlineMaterial=null
      outlineSize=0
      outlineColor=<rgba 0 0 0 1>
     }
    }
    shadowOffset=<0 0>
   }
   slot {class=HorizontalBoxSlot, properties {padding=<ltrb 0 0 20 0>, size {sizeRule=Fill, value=1}}}
  }
  Image_StatusIcon {
   widgetClassPath=&Image
   bIsVariable=true
   parent=HorizontalBox_3
   widget=Image_StatusIcon
   properties {brush {
     bIsDynamicallyLoaded=false
     drawAs=Image
     imageType=NoImage
     mirroring=NoMirror
     resourceName=None
     resourceObject=&M_UI_NeedsWeeds_1_NoShadow
     tiling=NoTile
     imageSize=<32 32>
     margin=<ltrb 0 0 0 0>
     outlineSettings {
      bUseBrushTransparency=false
      roundingType=HalfHeightRadius
      width=0
      color {colorUseRule=UseColor_Specified, specifiedColor=<rgba 0 0 0 0>}
      cornerRadii=<0 0 0 0>
     }
     tintColor {colorUseRule=UseColor_Specified, specifiedColor=<rgba 1 1 1 1>}
     uVRegion {bIsValid=false, max=<0 0>, min=<0 0>}
    }}
   slot {class=HorizontalBoxSlot, properties {horizontalAlignment=HAlign_Right}}
  }
  HorizontalBox_EnemyEvents {
   widgetClassPath=&HorizontalBox
   bIsVariable=true
   parent=VerticalBox_0
   widget=HorizontalBox_EnemyEvents
   properties {flowDirectionPreference=Culture, visibility=HitTestInvisible}
   slot {class=VerticalBoxSlot, properties {padding=<ltrb 0 0 0 10>}}
  }
  TextBlock_EnemyEvent {
   widgetClassPath=&CommonTextBlock
   bIsVariable=true
   parent=HorizontalBox_EnemyEvents
   widget=TextBlock_EnemyEvent
   properties {
    mobileFontSizeMultiplier=1.3
    style=&/RealitySaplingTextStyle_Subtext.RealitySaplingTextStyle_Subtext_C
    text=Under attack
    textTransformPolicy=ToUpper
    font {
     bForceMonospaced=false
     bMaterialIsStencil=false
     fontMaterial=null
     fontObject=&BurbankSmall
     letterSpacing=0
     monospacedWidth=1
     size=13
     skewAmount=0
     typefaceFontName=Bold
     outlineSettings {
      bApplyOutlineToDropShadows=false
      bMiteredCorners=false
      bSeparateFillAlpha=false
      outlineMaterial=null
      outlineSize=0
      outlineColor=<rgba 0 0 0 1>
     }
    }
    shadowOffset=<0 0>
   }
   slot {class=HorizontalBoxSlot, properties {verticalAlignment=VAlign_Center, size {sizeRule=Fill, value=1}}}
  }
  Image_EnemyEventIcon {
   widgetClassPath=&Image
   bIsVariable=true
   parent=HorizontalBox_EnemyEvents
   widget=Image_EnemyEventIcon
   properties {brush {
     bIsDynamicallyLoaded=false
     drawAs=Image
     imageType=NoImage
     mirroring=NoMirror
     resourceName=None
     resourceObject=/Game/UI/Foundation/Textures/Icons/GamepadOptions/CombatMode
     tiling=NoTile
     imageSize=<32 32>
     margin=<ltrb 0 0 0 0>
     outlineSettings {
      bUseBrushTransparency=false
      roundingType=HalfHeightRadius
      width=0
      color {colorUseRule=UseColor_Specified, specifiedColor=<rgba 0 0 0 0>}
      cornerRadii=<0 0 0 0>
     }
     tintColor {colorUseRule=UseColor_Specified, specifiedColor=<rgba 1 1 1 1>}
     uVRegion {bIsValid=false, max=<0 0>, min=<0 0>}
    }}
   slot {class=HorizontalBoxSlot, properties {horizontalAlignment=HAlign_Right}}
  }
  HorizontalBox_2 {
   widgetClassPath=&HorizontalBox
   parent=VerticalBox_0
   widget=HorizontalBox_2
   properties {flowDirectionPreference=Culture, visibility=HitTestInvisible}
   slot {class=VerticalBoxSlot, properties {padding=<ltrb 0 0 0 10>}}
  }
  VerticalBox_1 {
   widgetClassPath=&VerticalBox
   parent=HorizontalBox_2
   widget=VerticalBox_1
   slot {class=HorizontalBoxSlot, properties {size {sizeRule=Fill, value=1}}}
  }
  TextBlock_Fruit {
   widgetClassPath=&CommonTextBlock
   parent=VerticalBox_1
   widget=TextBlock_Fruit
   properties {
    mobileFontSizeMultiplier=1.3
    style=&RealitySaplingTextStyle_Subtext_C
    text=Fruit
    textTransformPolicy=ToUpper
    font {
     bForceMonospaced=false
     bMaterialIsStencil=false
     fontMaterial=null
     fontObject=&BurbankSmall
     letterSpacing=0
     monospacedWidth=1
     size=13
     skewAmount=0
     typefaceFontName=Bold
     outlineSettings {
      bApplyOutlineToDropShadows=false
      bMiteredCorners=false
      bSeparateFillAlpha=false
      outlineMaterial=null
      outlineSize=0
      outlineColor=<rgba 0 0 0 1>
     }
    }
    shadowOffset=<0 0>
   }
   slot {class=VerticalBoxSlot}
  }
  TextBlock_FruitRarity {
   widgetClassPath=&CommonTextBlock
   bIsVariable=true
   parent=VerticalBox_1
   widget=TextBlock_FruitRarity
   properties {
    lineHeightPercentage=0.95
    mobileFontSizeMultiplier=1.3
    text=Common
    textTransformPolicy=ToUpper
    font {
     bForceMonospaced=false
     bMaterialIsStencil=false
     fontMaterial=null
     fontObject=&BurbankSmall
     letterSpacing=0
     monospacedWidth=1
     size=12
     skewAmount=0
     typefaceFontName=Black
     outlineSettings {
      bApplyOutlineToDropShadows=false
      bMiteredCorners=false
      bSeparateFillAlpha=false
      outlineMaterial=null
      outlineSize=0
      outlineColor=<rgba 0 0 0 1>
     }
    }
    margin=<ltrb 0 -1 0 0>
    renderTransform {angle=0, scale=<1 1>, shear=<-12 0>, translation=<0 0>}
    shadowOffset=<0 0>
   }
   slot {class=VerticalBoxSlot}
  }
  HorizontalBox_Fruit {
   widgetClassPath=&HorizontalBox
   bIsVariable=true
   parent=HorizontalBox_2
   widget=HorizontalBox_Fruit
   slot {class=HorizontalBoxSlot, properties {horizontalAlignment=HAlign_Right, verticalAlignment=VAlign_Center}}
  }
  RealityFruit_147 {
   widgetClassPath=&/RealityFruit
   parent=HorizontalBox_Fruit
   widget=RealityFruit_147
   properties {iconSize=40}
   slot {class=HorizontalBoxSlot}
  }
  RealityFruit {
   widgetClassPath=&RealityFruit
   parent=HorizontalBox_Fruit
   widget=RealityFruit
   properties {iconSize=40}
   slot {class=HorizontalBoxSlot}
  }
  RealityFruit_300 {
   widgetClassPath=&RealityFruit
   parent=HorizontalBox_Fruit
   widget=RealityFruit_300
   properties {iconSize=40}
   slot {class=HorizontalBoxSlot}
  }
  HorizontalBox_NeedsWeeding {
   widgetClassPath=&HorizontalBox
   parent=VerticalBox_0
   widget=HorizontalBox_NeedsWeeding
   properties {flowDirectionPreference=Culture}
   slot {class=VerticalBoxSlot}
  }
  TextBlock_NextWeeding {
   widgetClassPath=&CommonTextBlock
   parent=HorizontalBox_NeedsWeeding
   widget=TextBlock_NextWeeding
   properties {
    mobileFontSizeMultiplier=1.3
    style=&RealitySaplingTextStyle_Subtext_C
    text=Next Weeding
    textTransformPolicy=ToUpper
    font {
     bForceMonospaced=false
     bMaterialIsStencil=false
     fontMaterial=null
     fontObject=&BurbankSmall
     letterSpacing=0
     monospacedWidth=1
     size=13
     skewAmount=0
     typefaceFontName=Bold
     outlineSettings {
      bApplyOutlineToDropShadows=false
      bMiteredCorners=false
      bSeparateFillAlpha=false
      outlineMaterial=null
      outlineSize=0
      outlineColor=<rgba 0 0 0 1>
     }
    }
    shadowOffset=<0 0>
   }
   slot {class=HorizontalBoxSlot, properties {verticalAlignment=VAlign_Center, padding=<ltrb 0 0 5 0>, size {sizeRule=Fill, value=1}}}
  }
  WidgetSwitcher_WeedingStatus {
   widgetClassPath=&WidgetSwitcher
   bIsVariable=true
   parent=HorizontalBox_NeedsWeeding
   widget=WidgetSwitcher_WeedingStatus
   slot {class=HorizontalBoxSlot, properties {horizontalAlignment=HAlign_Right, verticalAlignment=VAlign_Center, size {sizeRule=Fill, value=1}}}
  }
  RichText_NextWeedTime {
   widgetClassPath=&CommonTextBlock
   parent=WidgetSwitcher_WeedingStatus
   widget=RichText_NextWeedTime
   properties {
    mobileFontSizeMultiplier=1.3
    style=&RealitySaplingTextStyle_Subtext_C
    text="Now!"
    textTransformPolicy=ToUpper
    font {
     bForceMonospaced=false
     bMaterialIsStencil=false
     fontMaterial=null
     fontObject=&BurbankSmall
     letterSpacing=0
     monospacedWidth=1
     size=13
     skewAmount=0
     typefaceFontName=Bold
     outlineSettings {
      bApplyOutlineToDropShadows=false
      bMiteredCorners=false
      bSeparateFillAlpha=false
      outlineMaterial=null
      outlineSize=0
      outlineColor=<rgba 0 0 0 1>
     }
    }
    shadowOffset=<0 0>
   }
   slot {class=WidgetSwitcherSlot}
  }
  CommonDateTimeTextBlock_NextWeedTime {
   widgetClassPath=/Script/CommonUI.CommonDateTimeTextBlock
   bIsVariable=true
   parent=WidgetSwitcher_WeedingStatus
   widget=CommonDateTimeTextBlock_NextWeedTime
   properties {
    style=&RealitySaplingTextStyle_Subtext_C
    text=Text Block
    font {
     bForceMonospaced=false
     bMaterialIsStencil=false
     fontMaterial=null
     fontObject=&BurbankSmall
     letterSpacing=0
     monospacedWidth=1
     size=13
     skewAmount=0
     typefaceFontName=Bold
     outlineSettings {
      bApplyOutlineToDropShadows=false
      bMiteredCorners=false
      bSeparateFillAlpha=false
      outlineMaterial=null
      outlineSize=0
      outlineColor=<rgba 0 0 0 1>
     }
    }
    shadowOffset=<0 0>
   }
   slot {class=WidgetSwitcherSlot}
  }
  TutorialHighlight {widgetClassPath=&TutorialHighlight}
 ]
}
